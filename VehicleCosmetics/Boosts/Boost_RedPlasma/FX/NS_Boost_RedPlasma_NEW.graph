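# Niagara system for the Red Plasma vehicle boost cosmetic, drawing mesh and sprite exhaust at skeletal mesh sockets.
# Spawn groups act as layers within one emitter, setting render tag, lifetime, size, offset and motion through masked inputs.
# Particles are partitioned across sockets; cone-velocity streaks get pulled toward a point behind each socket.
# Painted-variant overrides and GameModeID switches set colors; the visible portion includes two disabled earlier emitter versions.
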NiagaraSystem /VehicleCosmetics/Boosts/Boost_RedPlasma/FX/NS_Boost_RedPlasma_NEW
dependencies {
 usedDataInterfaces [
  &NiagaraDataInterfaceCurve
  &NiagaraDataInterfaceColorCurve
  &NiagaraDataInterfaceVector2DCurve
  &NiagaraDataInterfaceVectorCurve
  &NiagaraDataInterfaceSkeletalMesh
 ]
 usedDynamicInputs [/Niagara/DynamicInputs/Helpers/MaskBoolBySpawnGroup]
 usedModules [
  &self:SystemSpawnScript.SystemScriptSource.SystemScriptGraph.NiagaraNodeAssignment_0.SetVariables_#1
  &SystemState
 ]
}
emitters {
 Meshes001 {
  bEnabled=false
  emitterUpdateScript [
   &EmitterState(Inactive Response=::NewEnumerator2, Life Cycle Mode=::NewEnumerator1, Loop Duration=1, UseLoopDelay=false)
   &VC_FilteredSocketCount()
   &SpawnRate(SpawnRate=30, Use Spawn Probability=false)
   &SpawnRate(SpawnGroup=1, SpawnRate=25, Use Spawn Probability=false) {module=SpawnRate001}
   &SpawnBurst_Instantaneous(Spawn Count=8, Spawn Group=3, Use Spawn Probability=false, UseLoopCountLimit=false)
   &SpawnRate(SpawnGroup=2, SpawnRate=120, Use Spawn Probability=false) {module=SpawnRate002}
   &SpawnRate(SpawnGroup=4, SpawnRate=120, Use Spawn Probability=false) {module=SpawnRate003}
  ]
  particleSpawnScript [
   /SetVariables(Particles.RenderTag:NiagaraInt32=&SpawnGroup()) {isSetParameters=true}
   &VC_PaintedVariantsOverride() {enabled=false}
   &InitializeParticle(
    Color Mode=::NewEnumerator0
    Lifetime=&MaskFloatBySpawnGroup(
     Masked Float=1
     Passthrough Float=&MaskFloatBySpawnGroup(
      Masked Float=1
      Passthrough Float=&MaskFloatBySpawnGroup(
       Masked Float=0.16
       Passthrough Float=&MaskFloatBySpawnGroup(
        Masked Float=1
        Passthrough Float=&MaskFloatBySpawnGroup(Masked Float=0.16, Passthrough Float=1, Spawn Group Mask=4)
        Spawn Group Mask=3
       )
       Spawn Group Mask=2
      )
      Spawn Group Mask=1
     )
    )
    Mesh Scale=&MaskVectorBySpawnGroup(
     Masked Vector=<0.2 0.2 0.34>
     Passthrough Vector=&MaskVectorBySpawnGroup(
      Masked Vector=<0.1 0.1 0.07>
      Passthrough Vector=&MaskVectorBySpawnGroup(
       Masked Vector=<0.2 0.2 0.2>
       Passthrough Vector=&MaskVectorBySpawnGroup(Masked Vector=<1 1 1>, Passthrough Vector=<1 1 1>, Spawn Group Mask=3)
       Spawn Group Mask=2
      )
      Spawn Group Mask=1
     )
    )
    Mesh Scale Mode=::NewEnumerator4
    Sprite Rotation Mode=::NewEnumerator2
    Sprite Size Max=&MaskVector2dBySpawnGroup(
     Masked Vector2D=<36 45>
     Passthrough Vector2D=&MaskVector2dBySpawnGroup(
      Masked Vector2D=<300 300>
      Passthrough Vector2D=&MaskVector2dBySpawnGroup(Masked Vector2D=<36 45>, Passthrough Vector2D=<1 1>, Spawn Group Mask=4)
      Spawn Group Mask=3
     )
     Spawn Group Mask=2
    )
    Sprite Size Min=&MaskVector2dBySpawnGroup(
     Masked Vector2D=<36 45>
     Passthrough Vector2D=&MaskVector2dBySpawnGroup(
      Masked Vector2D=<300 300>
      Passthrough Vector2D=&MaskVector2dBySpawnGroup(Masked Vector2D=<36 45>, Passthrough Vector2D=<1 1>, Spawn Group Mask=4)
      Spawn Group Mask=3
     )
     Spawn Group Mask=2
    )
    Sprite Size Mode=::NewEnumerator2
    UsePositionOffset=false
    Write Material Random=true
   )
   &PartitionParticles(Number Of Partitions=$Emitter.PortCount, Particles Per Partition=2, Partition Mode:ENiagara_PartitionMode=::NewEnumerator1)
   &InitialMeshOrientation(
    Mesh Axis To Orient=<0 0 0>
    Mesh Orientation Mode=::NewEnumerator2
    Orientation Vector=&MaskVectorBySpawnGroup(
     Masked Vector=<0 0 0.5>
     Passthrough Vector=&MaskVectorBySpawnGroup(
      Masked Vector=<0 0 0.5>
      Passthrough Vector=&MaskVectorBySpawnGroup(
       Masked Vector=<0 0 0.5>
       Passthrough Vector=&MaskVectorBySpawnGroup(Masked Vector=<0 0 0>, Passthrough Vector=<0 0 0>, Spawn Group Mask=3)
       Spawn Group Mask=2
      )
      Spawn Group Mask=1
     )
    )
    Rotation=&RandomRangeVector(
     Minimum=&MaskVectorBySpawnGroup(
      Masked Vector=<0 0 -1>
      Passthrough Vector=&MaskVectorBySpawnGroup(
       Masked Vector=<0 0 -1>
       Passthrough Vector=&MaskVectorBySpawnGroup(
        Masked Vector=<0 0 -1>
        Passthrough Vector=&MaskVectorBySpawnGroup(Masked Vector=<0 0 0>, Passthrough Vector=<1 1 1>, Spawn Group Mask=3)
        Spawn Group Mask=2
       )
       Spawn Group Mask=1
      )
     )
     Maximum=&MaskVectorBySpawnGroup(
      Masked Vector=<0 0 1>
      Passthrough Vector=&MaskVectorBySpawnGroup(
       Masked Vector=<0 0 1>
       Passthrough Vector=&MaskVectorBySpawnGroup(
        Masked Vector=<0 0 1>
        Passthrough Vector=&MaskVectorBySpawnGroup(Masked Vector=<0 0 0>, Passthrough Vector=<1 1 1>, Spawn Group Mask=3)
        Spawn Group Mask=2
       )
       Spawn Group Mask=1
      )
     )
    )
    Use Rotation Vector=true
   )
   &SkeletalMeshLocation(
    Bone / Socket Index=$Particles.PartitionParticles.Partition
    Bone / Socket Sampling Mode=::NewEnumerator5
    MaskSpawnGroup=false
    Mesh Sampling Type=::NewEnumerator6
    Sampled Position Offset=&AddVectorToPosition(
     Position=&MaskVectorBySpawnGroup(
      Masked Vector=<-50 0 0>
      Passthrough Vector=&MaskVectorBySpawnGroup(
       Masked Vector=<-46 0 0>
       Passthrough Vector=&MaskVectorBySpawnGroup(
        Masked Vector=<-45 0 0>
        Passthrough Vector=&MaskVectorBySpawnGroup(
         Masked Vector=<-110 0 0>
         Passthrough Vector=&MaskVectorBySpawnGroup(Masked Vector=<-75.9804 0 0>, Passthrough Vector=<1 1 1>, Spawn Group Mask=4)
         Spawn Group Mask=3
        )
        Spawn Group Mask=2
       )
       Spawn Group Mask=1
      )
     )
     Vector=Linked() {linkedVariable {name=User.DTCarOffset003, type=NiagaraPosition}}
    )
    Skeletal Mesh=$System.SkeletalMesh
   )
   &AddVelocity(Velocity=&MaskVectorBySpawnGroup(
     Masked Vector=<-30 0 0>
     Passthrough Vector=&MaskVectorBySpawnGroup(
      Masked Vector=<-5 0 0>
      Passthrough Vector=&MaskVectorBySpawnGroup(Masked Vector=<0 0 0>, Passthrough Vector=<1 1 1>, Spawn Group Mask=3)
      Spawn Group Mask=1
     )
    ))
   &AddVelocity(
    Cone Angle=&MaskFloatBySpawnGroup(
     Masked Float=&MergeHelper_FloatSwitch(0=140, 1=140, 2=140, 3=?, Index=$User.GameModeID, Num=::NewEnumerator1)
     Passthrough Float=&MaskFloatBySpawnGroup(
      Masked Float=&MergeHelper_FloatSwitch(0=125, 1=125, 2=125, 3=?, Index=$User.GameModeID, Num=::NewEnumerator1)
      Passthrough Float=1
      Spawn Group Mask=4
     )
     Spawn Group Mask=2
    )
    Distribution Along Cone Axis=0
    Module.UseConeFalloff=false
    Module.UseVelDistribution=true
    Override Seed=false
    Velocity Mode=::NewEnumerator2
    Velocity Speed=&MaskFloatBySpawnGroup(
     Masked Float=&MergeHelper_FloatSwitch(0=-350, 1=-350, 2=-350, 3=?, Index=$User.GameModeID, Num=::NewEnumerator1)
     Passthrough Float=&MaskFloatBySpawnGroup(
      Masked Float=&MergeHelper_FloatSwitch(0=-350, 1=-350, 2=-350, 3=?, Index=$User.GameModeID, Num=::NewEnumerator1)
      Passthrough Float=0
      Spawn Group Mask=4
     )
     Spawn Group Mask=2
    )
   ) {module=AddVelocity001}
   &AddVelocity(
    Cone Angle=&MergeHelper_FloatSwitch(0=140, 1=140, 2=140, 3=?, Index=$User.GameModeID, Num=::NewEnumerator1)
    Module.UseConeFalloff=false
    Module.UseVelDistribution=false
    Override Seed=false
    Velocity Mode=::NewEnumerator2
    Velocity Speed=&MaskFloatBySpawnGroup(
     Masked Float=&MergeHelper_FloatSwitch(0=-350, 1=-350, 2=-350, 3=?, Index=$User.GameModeID, Num=::NewEnumerator1)
     Passthrough Float=&MaskFloatBySpawnGroup(
      Masked Float=&MergeHelper_FloatSwitch(0=-350, 1=-350, 2=-350, 3=?, Index=$User.GameModeID, Num=::NewEnumerator1)
      Passthrough Float=0
      Spawn Group Mask=4
     )
     Spawn Group Mask=2
    )
   ) {enabled=false, module=AddVelocity002}
  ]
  particleUpdateScript [
   &ParticleState()
   &ScaleSpriteSize(Uniform Curve Sprite Scale=DataInterface() {propertyValues {Curve {keys [{tangentMode=RCTM_User, time=0, value=0}, <at 0.0988 1>, <at 0.95 1>, {tangentMode=RCTM_User, time=1, value=0}]}}}) {enabled=false}
   &ScaleSpriteSize(
    Uniform Curve Scale=&MaskFloatBySpawnGroup(Masked Float=1, Passthrough Float=1, Spawn Group Mask=2)
    Uniform Curve Sprite Scale=DataInterface() {propertyValues {Curve {keys [{tangentMode=RCTM_User, time=0, value=0}, <at 0.0988 1>, <at 0.95 1>, {tangentMode=RCTM_User, time=1, value=0}]}}}
   ) {enabled=false, module=ScaleSpriteSize001}
   &ScaleSpriteSize(
    Scale Factor=&MaskVector2dBySpawnGroup(
     Masked Vector2D=&Vector2DFromCurve(
      Vector2Curve=DataInterface() {propertyValues {XCurve {keys [<at 0 0>,<at 0.098 1>,<at 0.95 1>,<at 1 0>]}, YCurve {keys [<at 0 0>,<at 0.098 1>,<at 0.95 1>,<at 1 0>]}}}
      ScaleCurve=<1 1>
     )
     Passthrough Vector2D=&Vector2DFromCurve(
      Vector2Curve=DataInterface() {propertyValues {
        XCurve {
         defaultValue=0.5
         keys [
          {interpMode=RCIM_Cubic, tangentMode=RCTM_User, time=0, value=0.75}
          {interpMode=RCIM_Cubic, tangentMode=RCTM_User, time=0.5, value=0.88}
          {tangentMode=RCTM_User, time=1, value=0.75}
         ]
        }
        YCurve {
         defaultValue=0.5
         keys [
          {interpMode=RCIM_Cubic, tangentMode=RCTM_User, time=0, value=0.75}
          {interpMode=RCIM_Cubic, tangentMode=RCTM_User, time=0.5, value=0.88}
          {tangentMode=RCTM_User, time=1, value=0.75}
         ]
        }
       }}
      ScaleCurve=<1 1>
     )
     Spawn Group Mask=0
    )
    Scale Sprite Size Mode=::NewEnumerator2
   ) {module=ScaleSpriteSize002}
   &ScaleMeshSize(Scale Factor=&MaskVectorBySpawnGroup(
     Masked Vector=&VectorFromCurve(VectorCurve=DataInterface() {propertyValues {
        XCurve {keys [<at 0 1>,<at 0.25 1.25>,<at 1 0>]}
        YCurve {keys [<at 0 1>,<at 0.25 1.25>,<at 1 0>]}
        ZCurve {keys [<at 0 1>,<at 0.25 1.25>,<at 1 0>]}
       }})
     Passthrough Vector=&MaskVectorBySpawnGroup(
      Masked Vector=&VectorFromCurve(VectorCurve=DataInterface() {propertyValues {
         XCurve {keys [<at 0.0085 0.7461>,<at 0.1197 0.9841>,<at 0.3924 0.8409>,<at 0.9973 0.1466>]}
         YCurve {keys [<at 0.0017 0.7193>,<at 0.1289 0.9841>,<at 0.3954 0.8606>,<at 0.9973 0.1466>]}
         ZCurve {keys [<at 0.0022 0.7717>,<at 0.1517 1.1096>,<at 0.3853 1.0096>,<at 1.0156 0.1427>]}
        }})
      Passthrough Vector=&MaskVectorBySpawnGroup(
       Masked Vector=&VectorFromCurve(VectorCurve=DataInterface() {propertyValues {
          XCurve {keys [<at -0.0069 0.4922>,<at 0.0243 5>,<at 0.495 0.6299>,<at 0.5 0.57>,<at 0.9519 0.57>,<at 1 0>]}
          YCurve {keys [<at -0.0069 0.4922>,<at 0.0377 5>,<at 0.495 0.6299>,<at 0.5 0.57>,<at 0.9361 0.57>,<at 1 0>]}
          ZCurve {keys [<at 0 0.5668>,<at 0.0432 5>,<at 0.495 0.7017>,<at 0.5 0.78>,<at 0.8305 0.78>,<at 1.0139 0.75>]}
         }})
       Passthrough Vector=&MaskVectorBySpawnGroup(
        Masked Vector=&VectorFromFloat(Value=&FloatFromCurve(FloatCurve=DataInterface() {propertyValues {Curve {keys [
              <at 0 0.2824>
              {arriveTangent=0.049, interpMode=RCIM_Cubic, leaveTangent=0.049, tangentMode=RCTM_User, time=0.2283, value=0.9882}
              {arriveTangent=-1.3884, interpMode=RCIM_Cubic, leaveTangent=-1.3884, tangentMode=RCTM_User, time=0.6384, value=0.5912}
              {arriveTangent=-1.9181, tangentMode=RCTM_User, time=1, value=0}
             ]}}}))
        Passthrough Vector=<1 1 1>
        Spawn Group Mask=3
       )
       Spawn Group Mask=2
      )
      Spawn Group Mask=1
     )
    )) {module=ScaleMeshSize001}
   &PointAttractionForce(
    Attraction Radius=1000
    Attraction Strength=&MaskFloatBySpawnGroup(Masked Float=115, Passthrough Float=0, Spawn Group Mask=2)
    Attractor Position:NiagaraPosition=$StackContext.SkeletalMeshLocation.SampledPosition
    Attractor Position Offset:Vector3f=<-325 0 36.5386>
    Attractor Position Offset Coordinate Space:ENiagaraCoordinateSpace=::Local
    Kill Within Radius=false
    Module.Use Falloff=true
    UseAttractorOffset=true
   ) {module=PointAttractionForce001}
   &SolveForcesAndVelocity(Clamp Velocity=false, Limit Acceleration=false)
   &Color(Color=<rgba 0.75 0 0 1>)
   /SetVariables(
    Particles.BlackColor:LinearColor=<rgba 0 0 0 1>
    Particles.Int Fire:LinearColor=<rgba 4 0.2154 4 1>
    Particles.LinesColor:LinearColor=&ColorFromCurve(ColorCurve=DataInterface() {propertyValues {
       AlphaCurve {keys [<at 0.1329 0>,<at 0.2544 1>,<at 0.927 1 RCIM_Cubic>,<at 1 0 RCIM_Cubic>]}
       BlueCurve {keys [<at 1 0 RCIM_Cubic>]}
       GreenCurve {keys [<at 1 0 RCIM_Cubic>]}
       RedCurve {keys [<at 1 0 RCIM_Cubic>]}
      }})
    Particles.Ring Color:LinearColor=<rgba 3.3438 0 2.4158 1>
   ) {isSetParameters=true}
   &DynamicMaterialParameters(
    Index 0 Param 1=0.0087
    Index 0 Param 2=-35.9908
    Index 0 Param 3=-0.0477
    Index 1 Param 1=1
    Index 1 Param 2=1
    Module.Index0Float0Write=true
    Module.Index0Float1Write=true
    Module.Index0Float2Write=true
    Module.Index0Float3Write=true
    Module.Index1Float0Write=true
    Module.Index1Float1Write=true
    Module.Index1Float2Write=true
    Module.Index1Float3Write=true
    Write Parameter Index 1=::NewEnumerator1
   )
   &CameraOffset(Camera Offset Amount=&MaskFloatBySpawnGroup(Masked Float=-20, Passthrough Float=1, Spawn Group Mask=3))
  ]
  rendererClasses [&NiagaraMeshRendererProperties, &NiagaraSpriteRendererProperties]
  renderers [
   {rendererClass=&NiagaraMeshRendererProperties, rendererIndex=0}
   {rendererClass=&NiagaraMeshRendererProperties, rendererIndex=1}
   {rendererClass=&NiagaraSpriteRendererProperties, rendererIndex=2}
   {rendererClass=&NiagaraSpriteRendererProperties, rendererIndex=3}
   {rendererClass=&NiagaraSpriteRendererProperties, rendererIndex=4}
  ]
 }
 NE_VC_Boost_Cone_Spawn {
  bEnabled=false
  emitterUpdateScript [
   &EmitterState(Inactive Response=::NewEnumerator1, Life Cycle Mode=::NewEnumerator1, Loop Duration=1, UseLoopDelay=false)
   &VC_FilteredSocketCount()
   &SpawnRate(SpawnRate=30, Use Spawn Probability=false)
   &SpawnRate(SpawnGroup=1, SpawnRate=25, Use Spawn Probability=false) {module=SpawnRate001}
   &SpawnBurst_Instantaneous(Spawn Count=8, Spawn Group=3, Use Spawn Probability=false, UseLoopCountLimit=false)
   &SpawnRate(SpawnGroup=2, SpawnRate=120, Use Spawn Probability=false) {module=SpawnRate002}
   &SpawnRate(SpawnGroup=4, SpawnRate=120, Use Spawn Probability=false) {module=SpawnRate003}
  ]
  particleSpawnScript [
   /SetVariables(Particles.RenderTag:NiagaraInt32=&SpawnGroup()) {isSetParameters=true, module=SetVariables_#2}
   &VC_PaintedVariantsOverride(
    DefaultColor=&MergeHelper_ColorSwitch(0=<rgba 0.5322 0 0 1>, 1=<rgba 1 0.0185 0 1>, 2=<rgba 1 0.0185 0 1>, 3=?, Index=$User.GameModeID, Num=::NewEnumerator1)
    R1 DefaultColor=<rgba 0 0 0 1>
    R2 DefaultColor:LinearColor=<rgba 0 0 0 1>
    Use R1 Painted Overrides=true
    Use R2 Painted Overrides=true
   )
   &InitializeParticle(
    Color=$StackContext.Color
    Lifetime=&MaskFloatBySpawnGroup(
     Masked Float=1
     Passthrough Float=&MaskFloatBySpawnGroup(
      Masked Float=1
      Passthrough Float=&MaskFloatBySpawnGroup(
       Masked Float=0.16
       Passthrough Float=&MaskFloatBySpawnGroup(
        Masked Float=1
        Passthrough Float=&MaskFloatBySpawnGroup(Masked Float=0.16, Passthrough Float=1, Spawn Group Mask=4)
        Spawn Group Mask=3
       )
       Spawn Group Mask=2
      )
      Spawn Group Mask=1
     )
    )
    Mesh Scale=&MaskVectorBySpawnGroup(
     Masked Vector=<0.15 0.15 0.25>
     Passthrough Vector=&MaskVectorBySpawnGroup(
      Masked Vector=<0.08 0.08 0.055>
      Passthrough Vector=&MaskVectorBySpawnGroup(
       Masked Vector=<0.2 0.2 0.2>
       Passthrough Vector=&MaskVectorBySpawnGroup(Masked Vector=<1 1 1>, Passthrough Vector=<1 1 1>, Spawn Group Mask=3)
       Spawn Group Mask=2
      )
      Spawn Group Mask=1
     )
    )
    Mesh Scale Mode=::NewEnumerator4
    Sprite Rotation Mode=::NewEnumerator2
    Sprite Size Max=&MaskVector2dBySpawnGroup(
     Masked Vector2D=<0.25 1>
     Passthrough Vector2D=&MaskVector2dBySpawnGroup(
      Masked Vector2D=<250 250>
      Passthrough Vector2D=&MaskVector2dBySpawnGroup(Masked Vector2D=<20 45>, Passthrough Vector2D=<1 1>, Spawn Group Mask=4)
      Spawn Group Mask=3
     )
     Spawn Group Mask=2
    )
    Sprite Size Min=&MaskVector2dBySpawnGroup(
     Masked Vector2D=<0.25 1>
     Passthrough Vector2D=&MaskVector2dBySpawnGroup(
      Masked Vector2D=<250 250>
      Passthrough Vector2D=&MaskVector2dBySpawnGroup(Masked Vector2D=<20 45>, Passthrough Vector2D=<1 1>, Spawn Group Mask=4)
      Spawn Group Mask=3
     )
     Spawn Group Mask=2
    )
    Sprite Size Mode=::NewEnumerator2
    UsePositionOffset=false
    Write Material Random=true
   )
   &PartitionParticles(Number Of Partitions=$Emitter.PortCount, Particles Per Partition=2, Partition Mode:ENiagara_PartitionMode=::NewEnumerator1)
   &InitialMeshOrientation(
    Mesh Axis To Orient=<0 0 0>
    Mesh Orientation Mode=::NewEnumerator2
    Orientation Vector=&MaskVectorBySpawnGroup(
     Masked Vector=<0 0 0.5>
     Passthrough Vector=&MaskVectorBySpawnGroup(
      Masked Vector=<0 0 0.5>
      Passthrough Vector=&MaskVectorBySpawnGroup(
       Masked Vector=<0 0 0.5>
       Passthrough Vector=&MaskVectorBySpawnGroup(Masked Vector=<0 0 0>, Passthrough Vector=<0 0 0>, Spawn Group Mask=3)
       Spawn Group Mask=2
      )
      Spawn Group Mask=1
     )
    )
    Rotation=&RandomRangeVector(
     Minimum=&MaskVectorBySpawnGroup(
      Masked Vector=<0 0 -1>
      Passthrough Vector=&MaskVectorBySpawnGroup(
       Masked Vector=<0 0 -1>
       Passthrough Vector=&MaskVectorBySpawnGroup(
        Masked Vector=<0 0 -1>
        Passthrough Vector=&MaskVectorBySpawnGroup(Masked Vector=<0 0 0>, Passthrough Vector=<1 1 1>, Spawn Group Mask=3)
        Spawn Group Mask=2
       )
       Spawn Group Mask=1
      )
     )
     Maximum=&MaskVectorBySpawnGroup(
      Masked Vector=<0 0 1>
      Passthrough Vector=&MaskVectorBySpawnGroup(
       Masked Vector=<0 0 1>
       Passthrough Vector=&MaskVectorBySpawnGroup(
        Masked Vector=<0 0 1>
        Passthrough Vector=&MaskVectorBySpawnGroup(Masked Vector=<0 0 0>, Passthrough Vector=<1 1 1>, Spawn Group Mask=3)
        Spawn Group Mask=2
       )
       Spawn Group Mask=1
      )
     )
    )
    Use Rotation Vector=true
   )
   &SkeletalMeshLocation(
    Bone / Socket Index=$Particles.PartitionParticles.Partition
    Bone / Socket Sampling Mode=::NewEnumerator5
    Interpolate Mesh Sampling=false
    MaskSpawnGroup=false
    Mesh Sampling Type=::NewEnumerator6
    Sampled Position Offset=&AddVectorToPosition(
     Position=&MaskVectorBySpawnGroup(
      Masked Vector=<-50 0 0>
      Passthrough Vector=&MaskVectorBySpawnGroup(
       Masked Vector=<-46 0 0>
       Passthrough Vector=&MaskVectorBySpawnGroup(
        Masked Vector=<-45 0 0>
        Passthrough Vector=&MaskVectorBySpawnGroup(
         Masked Vector=<-110 0 0>
         Passthrough Vector=&MaskVectorBySpawnGroup(Masked Vector=<-75.9804 0 0>, Passthrough Vector=<1 1 1>, Spawn Group Mask=4)
         Spawn Group Mask=3
        )
        Spawn Group Mask=2
       )
       Spawn Group Mask=1
      )
     )
     Vector=Linked() {linkedVariable {name=User.DTCarOffset003, type=NiagaraPosition}}
    )
    Skeletal Mesh=$System.SkeletalMesh
   )
   &AddVelocity(Velocity=&MaskVectorBySpawnGroup(
     Masked Vector=<-30 0 0>
     Passthrough Vector=&MaskVectorBySpawnGroup(
      Masked Vector=<-3.5 0 0>
      Passthrough Vector=&MaskVectorBySpawnGroup(Masked Vector=<0 0 0>, Passthrough Vector=<1 1 1>, Spawn Group Mask=3)
      Spawn Group Mask=1
     )
    ))
   &AddVelocity(
    Cone Angle=&MaskFloatBySpawnGroup(
     Masked Float=&MergeHelper_FloatSwitch(0=150, 1=150, 2=150, 3=?, Index=$User.GameModeID, Num=::NewEnumerator1)
     Passthrough Float=&MaskFloatBySpawnGroup(
      Masked Float=&MergeHelper_FloatSwitch(0=125, 1=125, 2=125, 3=?, Index=$User.GameModeID, Num=::NewEnumerator1)
      Passthrough Float=1
      Spawn Group Mask=4
     )
     Spawn Group Mask=2
    )
    Distribution Along Cone Axis=0
    Module.UseConeFalloff=false
    Module.UseVelDistribution=true
    Override Seed=false
    Velocity Mode=::NewEnumerator2
    Velocity Speed=&MaskFloatBySpawnGroup(
     Masked Float=&MergeHelper_FloatSwitch(0=-250, 1=-250, 2=-250, 3=?, Index=$User.GameModeID, Num=::NewEnumerator1)
     Passthrough Float=&MaskFloatBySpawnGroup(
      Masked Float=&MergeHelper_FloatSwitch(0=-250, 1=-250, 2=-250, 3=?, Index=$User.GameModeID, Num=::NewEnumerator1)
      Passthrough Float=0
      Spawn Group Mask=4
     )
     Spawn Group Mask=2
    )
   ) {module=AddVelocity002}
  ]
  particleUpdateScript [
   &ParticleState()
   &ScaleMeshSize(Scale Factor=&MaskVectorBySpawnGroup(
     Masked Vector=&VectorFromCurve(VectorCurve=DataInterface() {propertyValues {
        XCurve {keys [<at 0 1>,<at 0.25 1.25>,<at 1 0>]}
        YCurve {keys [<at 0 1>,<at 0.25 1.25>,<at 1 0>]}
        ZCurve {keys [<at 0 1>,<at 0.25 1.25>,<at 1 0>]}
       }})
     Passthrough Vector=&MaskVectorBySpawnGroup(
      Masked Vector=&VectorFromCurve(VectorCurve=DataInterface() {propertyValues {
         XCurve {keys [<at 0.0085 0.7461>,<at 0.1197 0.9841>,<at 0.3924 0.8409>,<at 0.9973 0.1466>]}
         YCurve {keys [<at 0.0017 0.7193>,<at 0.1289 0.9841>,<at 0.3954 0.8606>,<at 0.9973 0.1466>]}
         ZCurve {keys [<at 0.0022 0.7717>,<at 0.1517 1.1096>,<at 0.3853 1.0096>,<at 1.0156 0.1427>]}
        }})
      Passthrough Vector=&MaskVectorBySpawnGroup(
       Masked Vector=&VectorFromCurve(VectorCurve=DataInterface() {propertyValues {
          XCurve {keys [<at -0.0069 0.4922>,<at 0.0243 5>,<at 0.495 0.6299>,<at 0.5 0.57>,<at 0.9519 0.57>,<at 1 0>]}
          YCurve {keys [<at -0.0069 0.4922>,<at 0.0377 5>,<at 0.495 0.6299>,<at 0.5 0.57>,<at 0.9361 0.57>,<at 1 0>]}
          ZCurve {keys [<at 0 0.5668>,<at 0.0432 5>,<at 0.495 0.7017>,<at 0.5 0.78>,<at 0.8305 0.78>,<at 1.0139 0.75>]}
         }})
       Passthrough Vector=&MaskVectorBySpawnGroup(
        Masked Vector=&VectorFromFloat(Value=&FloatFromCurve(FloatCurve=DataInterface() {propertyValues {Curve {keys [
              <at 0 0.2824>
              {arriveTangent=0.049, interpMode=RCIM_Cubic, leaveTangent=0.049, tangentMode=RCTM_User, time=0.2283, value=0.9882}
              {arriveTangent=-1.3884, interpMode=RCIM_Cubic, leaveTangent=-1.3884, tangentMode=RCTM_User, time=0.6384, value=0.5912}
              {arriveTangent=-1.9181, tangentMode=RCTM_User, time=1, value=0}
             ]}}}))
        Passthrough Vector=<1 1 1>
        Spawn Group Mask=3
       )
       Spawn Group Mask=2
      )
      Spawn Group Mask=1
     )
    )) {module=ScaleMeshSize001}
   &ScaleSpriteSize(
    Scale Factor=&MaskVector2dBySpawnGroup(
     Masked Vector2D=&Vector2DFromCurve(
      Vector2Curve=DataInterface() {propertyValues {
        XCurve {keys [<at 0 0>,<at 0.1304 0>,<at 0.2393 40>,<at 0.95 40>,<at 1 0>]}
        YCurve {keys [<at 0 0>,<at 0.1304 0>,<at 0.2417 40>,<at 0.95 40>,<at 1 0>]}
       }}
      ScaleCurve=<1 1>
     )
     Passthrough Vector2D=&Vector2DFromCurve(
      Vector2Curve=DataInterface() {propertyValues {
        XCurve {
         defaultValue=0.5
         keys [
          {interpMode=RCIM_Cubic, tangentMode=RCTM_User, time=0, value=0.75}
          {interpMode=RCIM_Cubic, tangentMode=RCTM_User, time=0.5, value=0.88}
          {tangentMode=RCTM_User, time=1, value=0.75}
         ]
        }
        YCurve {
         defaultValue=0.5
         keys [
          {interpMode=RCIM_Cubic, tangentMode=RCTM_User, time=0, value=0.75}
          {interpMode=RCIM_Cubic, tangentMode=RCTM_User, time=0.5, value=0.88}
          {tangentMode=RCTM_User, time=1, value=0.75}
         ]
        }
       }}
      ScaleCurve=<1 1>
     )
     Spawn Group Mask=2
    )
    Scale Sprite Size Mode=::NewEnumerator2
   ) {module=ScaleSpriteSize002}
   &PointAttractionForce(
    Attraction Radius=1000
    Attraction Strength=&MaskFloatBySpawnGroup(Masked Float=115, Passthrough Float=0, Spawn Group Mask=2)
    Attractor Position:NiagaraPosition=$StackContext.SkeletalMeshLocation.SampledPosition
    Attractor Position Offset:Vector3f=<-325 0 36>
    Attractor Position Offset Coordinate Space:ENiagaraCoordinateSpace=::Local
    Kill Within Radius=false
    Module.Use Falloff=true
    UseAttractorOffset=true
   ) {module=PointAttractionForce001}
   &SolveForcesAndVelocity(Clamp Velocity=false, Limit Acceleration=false)
   &Color(Color=&MaskLinearColorBySpawnGroup(Masked Linear Color=<rgba 0 0 0 1>, Passthrough Linear Color=<rgba 1 1 1 1>, Spawn Group Mask=2)) {enabled=false}
   /SetVariables(
    Particles.BlackColor:LinearColor=<rgba 0 0 0 1>
    Particles.Int Fire:LinearColor=<rgba 4 0.2154 4 1>
    Particles.LinesColor:LinearColor=&ColorFromCurve(ColorCurve=DataInterface() {propertyValues {
       AlphaCurve {keys [<at 0.1329 0>,<at 0.2544 1>,<at 0.927 1 RCIM_Cubic>,<at 1 0 RCIM_Cubic>]}
       BlueCurve {keys [<at 1 0 RCIM_Cubic>]}
       GreenCurve {keys [<at 1 0 RCIM_Cubic>]}
       RedCurve {keys [<at 1 0 RCIM_Cubic>]}
      }})
    Particles.Ring Color:LinearColor=<rgba 3.3438 0 2.4158 1>
   ) {enabled=false, isSetParameters=true}
   &ScaleColor(
    Scale Alpha=&MaskFloatBySpawnGroup(
     Masked Float=&FloatFromCurve(FloatCurve=DataInterface() {propertyValues {Curve {keys [<at 0 0>,<at 1 1>]}}})
     Passthrough Float=1
     Spawn Group Mask=2
    )
    Scale RGB=&Multiply_VectorByFloat(Vector=<1 1 1>, Float=1)
    ScaleA=true
    ScaleRGB=false
   ) {enabled=false}
   &DynamicMaterialParameters(
    Index 0 Param 1=0.0087
    Index 0 Param 2=-35.9908
    Index 0 Param 3=-0.0477
    Index 1 Param 1=1
    Index 1 Param 2=1
    Module.Index0Float0Write=true
    Module.Index0Float1Write=true
    Module.Index0Float2Write=true
    Module.Index0Float3Write=true
    Module.Index1Float0Write=true
    Module.Index1Float1Write=true
    Module.Index1Float2Write=true
    Module.Index1Float3Write=true
    Write Parameter Index 1=::NewEnumerator1
   )
   &CameraOffset(Camera Offset Amount=&MaskFloatBySpawnGroup(Masked Float=-20, Passthrough Float=1, Spawn Group Mask=3))
   /SetVariables(Particles.Lines:LinearColor=/DealershipSharedLibrary/FX/Niagara/DynamicInputs/MultiplyLinearColorByVectorAndFloat(
     Color to Scale:LinearColor=$Particles.Initial.R2 Color
     RGB Scale:Vector3f=<1 1 1>
     Alpha Scale:NiagaraFloat=&FloatFromCurve(FloatCurve=DataInterface() {propertyValues {Curve {keys [<at 0 0.031>,<at 0.4924 0.9131>,<at 1 0>]}}})
    )) {isSetParameters=true, module=SetVariables_#3}
  ]
  rendererClasses [&NiagaraMeshRendererProperties, &NiagaraSpriteRendererProperties]
  renderers [
   {rendererClass=&NiagaraMeshRendererProperties, rendererIndex=0}
   {rendererClass=&NiagaraMeshRendererProperties, rendererIndex=1}
   {rendererClass=&NiagaraSpriteRendererProperties, rendererIndex=2}
   {rendererClass=&NiagaraSpriteRendererProperties, rendererIndex=3}
   {rendererClass=&NiagaraSpriteRendererProperties, rendererIndex=4}
   {rendererClass=&NiagaraSpriteRendererProperties, rendererIndex=5}
  ]
 }
 NE_VC_Boost_Cone_Spawn001 {
  emitterUpdateScript [
   &EmitterState(Inactive Response=::NewEnumerator1, Life Cycle Mode=::NewEnumerator1, Loop Duration=1, UseLoopDelay=false)
   &VC_FilteredSocketCount()
   &SpawnRate(SpawnRate=30, Use Spawn Probability=false)
   &SpawnRate(SpawnGroup=1, SpawnRate=25, Use Spawn Probability=false) {module=SpawnRate001}
   &SpawnBurst_Instantaneous(Spawn Count=8, Spawn Group=3, Use Spawn Probability=false, UseLoopCountLimit=false)
   &SpawnRate(SpawnGroup=2, SpawnRate=120, Use Spawn Probability=false) {module=SpawnRate002}
   &SpawnRate(SpawnGroup=4, SpawnRate=120, Use Spawn Probability=false) {module=SpawnRate003}
  ]
  particleSpawnScript [
   /SetVariables(Particles.RenderTag:NiagaraInt32=&SpawnGroup()) {isSetParameters=true, module=SetVariables_#2}
   &VC_PaintedVariantsOverride(
    Black=<rgba 0.0281 0.0281 0.0281 1>
    BurntSienna=<rgba 0.06 0.0131 0.0073 1>
    Cobalt=<rgba 0.0019 0.0229 0.374 1>
    Crimson=<rgba 0.3165 0 0.0086 1>
    DefaultColor=&MergeHelper_ColorSwitch(0=<rgba 0.5322 0 0 1>, 1=<rgba 1 0.0185 0 1>, 2=<rgba 1 0.0185 0 1>, 3=?, Index=$User.GameModeID, Num=::NewEnumerator1)
    ForestGreen=<rgba 0 0.3073 0 1>
    Grey=<rgba 0.0823 0.0823 0.0823 1>
    Lime=<rgba 0.2893 0.449 0 1>
    Orange=<rgba 0.527 0.0425 0 1>
    Purple=<rgba 0.062 0 0.3073 1>
    R1 Black:LinearColor=<rgba 0.0191 0.0191 0.0191 1.5>
    R1 BurntSienna:LinearColor=<rgba 0.0642 0.0198 0.0079 0.25>
    R1 Cobalt:LinearColor=<rgba 0 0.0185 1 0.36>
    R1 Crimson:LinearColor=<rgba 0.4635 0 0.0126 0.2974>
    R1 DefaultColor=&MergeHelper_ColorSwitch(0=<rgba 0.3823 0 0 1>, 1=<rgba 1 0.0185 0 1>, 2=<rgba 1 0.0185 0 1>, 3=?, Index=$User.GameModeID, Num=::NewEnumerator1)
    R1 ForestGreen:LinearColor=<rgba 0 0.5 0.0253 0.2348>
    R1 Grey:LinearColor=<rgba 0.2031 0.2031 0.2031 0.2696>
    R1 Lime:LinearColor=<rgba 0.2943 0.5885 0 0.35>
    R1 Orange:LinearColor=<rgba 0.6724 0.2346 0 0.367>
    R1 Pink:LinearColor=<rgba 0.5052 0.0605 0.3435 0.367>
    R1 Purple:LinearColor=<rgba 0.2406 0.0148 0.6795 0.36>
    R1 Saffron:LinearColor=<rgba 0.9 0.69 0 0.25>
    R1 SkyBlue:LinearColor=<rgba 0 0.357 0.5365 0.2696>
    R1 TitaniumWhite:LinearColor=<rgba 0.75 0.75 0.75 0.2417>
    R2 Black:LinearColor=<rgba 0 0 0 1>
    R2 BurntSienna:LinearColor=<rgba 0 0 0 1>
    R2 Cobalt:LinearColor=<rgba 0 0 0 1>
    R2 Crimson:LinearColor=<rgba 0 0 0 1>
    R2 DefaultColor:LinearColor=<rgba 0 0 0 1>
    R2 ForestGreen:LinearColor=<rgba 0 0 0 1>
    R2 Gold:LinearColor=<rgba 0 0 0 1>
    R2 Grey:LinearColor=<rgba 0 0 0 1>
    R2 Lime:LinearColor=<rgba 0 0 0 1>
    R2 Orange:LinearColor=<rgba 0 0 0 1>
    R2 Pink:LinearColor=<rgba 0 0 0 1>
    R2 Purple:LinearColor=<rgba 0 0 0 1>
    R2 Saffron:LinearColor=<rgba 0 0 0 1>
    R2 SkyBlue:LinearColor=<rgba 0 0 0 1>
    R2 TitaniumWhite:LinearColor=<rgba 0 0 0 1>
    Saffron=<rgba 0.674 0.5254 0.0372 1>
    SkyBlue=<rgba 0 0.2076 0.574 1>
    TitaniumWhite=<rgba 0.567 0.567 0.567 1>
    Use R1 Painted Overrides=true
    Use R2 Painted Overrides=true
   )
   &InitializeParticle(
    Color=$StackContext.Color
    Lifetime=&MaskFloatBySpawnGroup(
     Masked Float=1
     Passthrough Float=&MaskFloatBySpawnGroup(
      Masked Float=1
      Passthrough Float=&MaskFloatBySpawnGroup(
       Masked Float=0.16
       Passthrough Float=&MaskFloatBySpawnGroup(
        Masked Float=1
        Passthrough Float=&MaskFloatBySpawnGroup(Masked Float=0.16, Passthrough Float=1, Spawn Group Mask=4)
        Spawn Group Mask=3
       )
       Spawn Group Mask=2
      )
      Spawn Group Mask=1
     )
    )
    Mesh Scale=&MaskVectorBySpawnGroup(
     Masked Vector=<0.15 0.15 0.25>
     Passthrough Vector=&MaskVectorBySpawnGroup(
      Masked Vector=<0.08 0.08 0.055>
      Passthrough Vector=&MaskVectorBySpawnGroup(
       Masked Vector=<0.2 0.2 0.2>
       Passthrough Vector=&MaskVectorBySpawnGroup(Masked Vector=<1 1 1>, Passthrough Vector=<1 1 1>, Spawn Group Mask=3)
       Spawn Group Mask=2
      )
      Spawn Group Mask=1
     )
    )
    Mesh Scale Mode=::NewEnumerator4
    Sprite Rotation Mode=::NewEnumerator2
    Sprite Size Max=&MaskVector2dBySpawnGroup(
     Masked Vector2D=<0.25 1>
     Passthrough Vector2D=&MaskVector2dBySpawnGroup(
      Masked Vector2D=<250 250>
      Passthrough Vector2D=&MaskVector2dBySpawnGroup(Masked Vector2D=<20 45>, Passthrough Vector2D=<1 1>, Spawn Group Mask=4)
      Spawn Group Mask=3
     )
     Spawn Group Mask=2
    )
    Sprite Size Min=&MaskVector2dBySpawnGroup(
     Masked Vector2D=<0.25 1>
     Passthrough Vector2D=&MaskVector2dBySpawnGroup(
      Masked Vector2D=<250 250>
      Passthrough Vector2D=&MaskVector2dBySpawnGroup(Masked Vector2D=<20 45>, Passthrough Vector2D=<1 1>, Spawn Group Mask=4)
      Spawn Group Mask=3
     )
     Spawn Group Mask=2
    )
    Sprite Size Mode=::NewEnumerator2
    UsePositionOffset=false
    Write Material Random=true
   )
   &PartitionParticles(Number Of Partitions=$Emitter.PortCount, Particles Per Partition=2, Partition Mode:ENiagara_PartitionMode=::NewEnumerator1)
   &InitialMeshOrientation(
    Mesh Axis To Orient=<0 0 0>
    Mesh Orientation Mode=::NewEnumerator2
    Orientation Vector=&MaskVectorBySpawnGroup(
     Masked Vector=<0 0 0.5>
     Passthrough Vector=&MaskVectorBySpawnGroup(
      Masked Vector=<0 0 0.5>
      Passthrough Vector=&MaskVectorBySpawnGroup(
       Masked Vector=<0 0 0.5>
       Passthrough Vector=&MaskVectorBySpawnGroup(Masked Vector=<0 0 0>, Passthrough Vector=<0 0 0>, Spawn Group Mask=3)
       Spawn Group Mask=2
      )
      Spawn Group Mask=1
     )
    )
    Rotation=&RandomRangeVector(
     Minimum=&MaskVectorBySpawnGroup(
      Masked Vector=<0 0 -1>
      Passthrough Vector=&MaskVectorBySpawnGroup(
       Masked Vector=<0 0 -1>
       Passthrough Vector=&MaskVectorBySpawnGroup(
        Masked Vector=<0 0 -1>
        Passthrough Vector=&MaskVectorBySpawnGroup(Masked Vector=<0 0 0>, Passthrough Vector=<1 1 1>, Spawn Group Mask=3)
        Spawn Group Mask=2
       )
       Spawn Group Mask=1
      )
     )
     Maximum=&MaskVectorBySpawnGroup(
      Masked Vector=<0 0 1>
      Passthrough Vector=&MaskVectorBySpawnGroup(
       Masked Vector=<0 0 1>
       Passthrough Vector=&MaskVectorBySpawnGroup(
        Masked Vector=<0 0 1>
        Passthrough Vector=&MaskVectorBySpawnGroup(Masked Vector=<0 0 0>, Passthrough Vector=<1 1 1>, Spawn Group Mask=3)
        Spawn Group Mask=2
       )
       Spawn Group Mask=1
      )
     )
    )
    Use Rotation Vector=true
   )
   &SkeletalMeshLocation(
    Bone / Socket Index=$Particles.PartitionParticles.Partition
    Bone / Socket Sampling Mode=::NewEnumerator5
    Interpolate Mesh Sampling=false
    MaskSpawnGroup=false
    Mesh Sampling Type=::NewEnumerator6
    Sampled Position Offset=&AddVectorToPosition(
     Position=&MaskVectorBySpawnGroup(
      Masked Vector=<-50 0 0>
      Passthrough Vector=&MaskVectorBySpawnGroup(
       Masked Vector=<-46 0 0>
       Passthrough Vector=&MaskVectorBySpawnGroup(
        Masked Vector=<-45 0 0>
        Passthrough Vector=&MaskVectorBySpawnGroup(
         Masked Vector=<-110 0 0>
         Passthrough Vector=&MaskVectorBySpawnGroup(Masked Vector=<-75.9804 0 0>, Passthrough Vector=<1 1 1>, Spawn Group Mask=4)
         Spawn Group Mask=3
        )
        Spawn Group Mask=2
       )
       Spawn Group Mask=1
      )
     )
     Vector=Linked() {linkedVariable {name=User.DTCarOffset003, type=NiagaraPosition}}
    )
    Skeletal Mesh=$System.SkeletalMesh
   )
   &AddVelocity(Velocity=&MaskVectorBySpawnGroup(
     Masked Vector=<-30 0 0>
     Passthrough Vector=&MaskVectorBySpawnGroup(
      Masked Vector=<-3.5 0 0>
      Passthrough Vector=&MaskVectorBySpawnGroup(Masked Vector=<0 0 0>, Passthrough Vector=<1 1 1>, Spawn Group Mask=3)
      Spawn Group Mask=1
     )
    ))
   &AddVelocity(
    Cone Angle=&MaskFloatBySpawnGroup(
     Masked Float=&MergeHelper_FloatSwitch(0=150, 1=150, 2=150, 3=?, Index=$User.GameModeID, Num=::NewEnumerator1)
     Passthrough Float=&MaskFloatBySpawnGroup(
      Masked Float=&MergeHelper_FloatSwitch(0=125, 1=125, 2=125, 3=?, Index=$User.GameModeID, Num=::NewEnumerator1)
      Passthrough Float=1
      Spawn Group Mask=4
     )
     Spawn Group Mask=2
    )
    Distribution Along Cone Axis=0
    Module.UseConeFalloff=false
    Module.UseVelDistribution=true
    Override Seed=false
    Velocity Mode=::NewEnumerator2
    Velocity Speed=&MaskFloatBySpawnGroup(
     Masked Float=&MergeHelper_FloatSwitch(0=-250, 1=-250, 2=-250, 3=?, Index=$User.GameModeID, Num=::NewEnumerator1)
     Passthrough Float=&MaskFloatBySpawnGroup(
      Masked Float=&MergeHelper_FloatSwitch(0=-250, 1=-250, 2=-250, 3=?, Index=$User.GameModeID, Num=::NewEnumerator1)
      Passthrough Float=0
      Spawn Group Mask=4
     )
     Spawn Group Mask=2
    )
   ) {module=AddVelocity002}
  ]
  particleUpdateScript [
   &ParticleState()
   &ScaleMeshSize(Scale Factor=&MaskVectorBySpawnGroup(
     Masked Vector=&VectorFromCurve(VectorCurve=DataInterface() {propertyValues {
        XCurve {keys [<at 0 1>,<at 0.25 1.25>,<at 1 0>]}
        YCurve {keys [<at 0 1>,<at 0.25 1.25>,<at 1 0>]}
        ZCurve {keys [<at 0 1>,<at 0.25 1.25>,<at 1 0>]}
       }})
     Passthrough Vector=&MaskVectorBySpawnGroup(
      Masked Vector=&VectorFromCurve(VectorCurve=DataInterface() {propertyValues {
         XCurve {keys [<at 0.0085 0.7461>,<at 0.1197 0.9841>,<at 0.3924 0.8409>,<at 0.9973 0.1466>]}
         YCurve {keys [<at 0.0017 0.7193>,<at 0.1289 0.9841>,<at 0.3954 0.8606>,<at 0.9973 0.1466>]}
         ZCurve {keys [<at 0.0022 0.7717>,<at 0.1517 1.1096>,<at 0.3853 1.0096>,<at 1.0156 0.1427>]}
        }})
      Passthrough Vector=&MaskVectorBySpawnGroup(
       Masked Vector=&VectorFromCurve(VectorCurve=DataInterface() {propertyValues {
          XCurve {keys [<at -0.0069 0.4922>,<at 0.0243 5>,<at 0.495 0.6299>,<at 0.5 0.57>,<at 0.9519 0.57>,<at 1 0>]}
          YCurve {keys [<at -0.0069 0.4922>,<at 0.0377 5>,<at 0.495 0.6299>,<at 0.5 0.57>,<at 0.9361 0.57>,<at 1 0>]}
          ZCurve {keys [<at 0 0.5668>,<at 0.0432 5>,<at 0.495 0.7017>,<at 0.5 0.78>,<at 0.8305 0.78>,<at 1.0139 0.75>]}
         }})
       Passthrough Vector=&MaskVectorBySpawnGroup(
        Masked Vector=&VectorFromFloat(Value=&FloatFromCurve(FloatCurve=DataInterface() {propertyValues {Curve {keys [
              <at 0 0.2824>
              {arriveTangent=0.049, interpMode=RCIM_Cubic, leaveTangent=0.049, tangentMode=RCTM_User, time=0.2283, value=0.9882}
              {arriveTangent=-1.3884, interpMode=RCIM_Cubic, leaveTangent=-1.3884, tangentMode=RCTM_User, time=0.6384, value=0.5912}
              {arriveTangent=-1.9181, tangentMode=RCTM_User, time=1, value=0}
             ]}}}))
        Passthrough Vector=<1 1 1>
        Spawn Group Mask=3
       )
       Spawn Group Mask=2
      )
      Spawn Group Mask=1
     )
    )) {module=ScaleMeshSize001}
   &ScaleSpriteSize(
    Scale Factor=&MaskVector2dBySpawnGroup(
     Masked Vector2D=&Vector2DFromCurve(
      Vector2Curve=DataInterface() {propertyValues {
        XCurve {keys [<at 0 0>,<at 0.1304 0>,<at 0.2393 40>,<at 0.95 40>,<at 1 0>]}
        YCurve {keys [<at 0 0>,<at 0.1304 0>,<at 0.2417 40>,<at 0.95 40>,<at 1 0>]}
       }}
      ScaleCurve=<1 1>
     )
     Passthrough Vector2D=&Vector2DFromCurve(
      Vector2Curve=DataInterface() {propertyValues {
        XCurve {
         defaultValue=0.5
         keys [
          {interpMode=RCIM_Cubic, tangentMode=RCTM_User, time=0, value=0.75}
          {interpMode=RCIM_Cubic, tangentMode=RCTM_User, time=0.5, value=0.88}
          {tangentMode=RCTM_User, time=1, value=0.75}
         ]
        }
        YCurve {
         defaultValue=0.5
         keys [
          {interpMode=RCIM_Cubic, tangentMode=RCTM_User, time=0, value=0.75}
          {interpMode=RCIM_Cubic, tangentMode=RCTM_User, time=0.5, value=0.88}
          {tangentMode=RCTM_User, time=1, value=0.75}
         ]
        }
       }}
      ScaleCurve=<1 1>
     )
     Spawn Group Mask=2
    )
    Scale Sprite Size Mode=::NewEnumerator2
   ) {module=ScaleSpriteSize002}
   &PointAttractionForce(
    Attraction Radius=1000
    Attraction Strength=&MaskFloatBySpawnGroup(Masked Float=115, Passthrough Float=0, Spawn Group Mask=2)
    Attractor Position:NiagaraPosition=$StackContext.SkeletalMeshLocation.SampledPosition
    Attractor Position Offset:Vector3f=<-325 0 36>
    Attractor Position Offset Coordinate Space:ENiagaraCoordinateSpace=::Local
    Kill Within Radius=false
    Module.Use Falloff=true
    UseAttractorOffset=true
   ) {module=PointAttractionForce001}
   &SolveForcesAndVelocity(Clamp Velocity=false, Limit Acceleration=false)
   &DynamicMaterialParameters(
    Index 0 Param 1=0.0087
    Index 0 Param 2=-35.9908
    Index 0 Param 3=-0.0477
    Index 1 Param 1=1
    Index 1 Param 2=1
    Module.Index0Float0Write=true
    Module.Index0Float1Write=true
    Module.Index0Float2Write=true
    Module.Index0Float3Write=true
    Module.Index1Float0Write=true
    Module.Index1Float1Write=true
    Module.Index1Float2Write=true
    Module.Index1Float3Write=true
    Write Parameter Index 1=::NewEnumerator1
   )
   &CameraOffset(Camera Offset Amount=&MaskFloatBySpawnGroup(Masked Float=-20, Passthrough Float=1, Spawn Group Mask=3))
   /SetVariables(Particles.Lines:LinearColor=&MultiplyLinearColorByVectorAndFloat(
     Color to Scale:LinearColor=$Particles.Initial.R2 Color
     RGB Scale:Vector3f=<1 1 1>
     Alpha Scale:NiagaraFloat=&FloatFromCurve(FloatCurve=DataInterface() {propertyValues {Curve {keys [<at 0 0.031>,<at 0.4924 0.9131>,<at 1 0>]}}})
    )) {isSetParameters=true, module=SetVariables_#3}
  ]
  rendererClasses [&NiagaraMeshRendererProperties, &NiagaraSpriteRendererProperties]
  renderers [
   {rendererClass=&NiagaraMeshRendererProperties, rendererIndex=0}
   {rendererClass=&NiagaraMeshRendererProperties, rendererIndex=1}
   {rendererClass=&NiagaraSpriteRendererProperties, rendererIndex=2}
   {rendererClass=&NiagaraSpriteRendererProperties, rendererIndex=3}
   {rendererClass=&NiagaraSpriteRendererProperties, rendererIndex=4}
   {rendererClass=&NiagaraSpriteRendererProperties, rendererIndex=5}
  ]
 }
 Ribbon_Electricity001 {
  bEnabled=false
  emitterSpawnScript [/SetVariables(
    Emitter.IsPainted:NiagaraFloat=&MakeCustomFloatFromBool(Bool=$User.bIsPainted, True Float=2, False Float=-0.512)
    Emitter.MirrorOffset:NiagaraFloat=&MakeCustomFloatFromBool(Bool=$User.bIsPainted, True Float=4, False Float=2)
   ) {isSetParameters=true, module=SetVariables_#4}]
  emitterUpdateScript [
   &EmitterState(Life Cycle Mode=::NewEnumerator1, Loop Duration=1, UseLoopDelay=false)
   &VC_FilteredSocketCount()
   &VC_SpawnRibbon_v1(
    Number of Ribbons=$Emitter.PortCount
    SpawnRate=&MakeCustomFloatFromBool(Bool=$User.bIsFrontendPreview, True Float=60, False Float=60)
   )
  ]
  particleSpawnScript [
   &VC_PaintedVariantsOverride() {enabled=false}
   &VC_InitializeRibbon_v1(
    Color=&MakeCustomLinearColorFromBool(Bool=$User.bIsPainted, True Color=$User.BelowTrailColor, False Color=<rgba 18.2664 5.7001 28.6436 1>)
    Custom Offset=true
    Enable Skeletal Position=true
    Lifetime=&MergeHelper_FloatSwitch(0=0.8, 1=0.8, 2=0.5, 3=?, Index=$User.GameModeID, Num=::NewEnumerator1)
    Local Vector Offset=&AddVectorToPosition(
     Position=&MergeHelper_VectorSwitch(0=<0 0 1>, 1=<0 0 1>, 2=<-58 0 1>, 3=?, Index=$System.GameModeID, Num=::NewEnumerator1)
     Vector=Linked() {linkedVariable {name=User.DTCarOffset003, type=NiagaraPosition}}
    )
    Offset Amount:NiagaraFloat=&MakeFloatFromVector(Vector=<0 0 0>, Channel=::NewEnumerator2)
    Ribbon Width=60
   )
   &AddVelocity(
    Velocity=&MakeCustomVectorFromBool(Bool=$User.bIsFrontendPreview, True Vector=<1000 0 0>, False Vector=<1000 0 0>)
    Velocity Speed Scale=-2
   ) {module=AddVelocity001}
   &ShapeLocation(MaskSpawnGroup=false, Override Seed=false, Sphere Radius=7, UseOwnerScale=true) {enabled=false}
  ]
  particleUpdateScript [
   &ParticleState()
   &InheritSourceMovement(
    Applied Position Scale=&Multiply_VectorByFloat(
     Vector=<1 1 1>
     Float=&FloatFromCurve(FloatCurve=DataInterface() {propertyValues {Curve {keys [{interpMode=RCIM_Cubic, tangentMode=RCTM_Break, time=0, value=0.75}, {interpMode=RCIM_Cubic, tangentMode=RCTM_Break, time=1, value=0}]}}})
    )
    ApplyForce=false
    ApplyPosition=true
    ApplyPositionFalloff=false
   )
   &ScaleColor(
    Scale Alpha=&MakeCustomFloatFromBool(
     Bool=$User.bIsPainted
     True Float=0.15
     False Float=&FloatFromCurve(FloatCurve=DataInterface() {propertyValues {Curve {keys [<at 0 0>,<at 0.15 0.5>,<at 0.7195 0>,<at 1 0>]}}}, Scale Curve=0.2)
    )
    Scale RGB=&MakeCustomVectorFromBool(Bool=$User.bIsPainted, True Vector=<4 4 4>, False Vector=<4 4 4>)
    ScaleA=true
    ScaleRGB=true
   )
   &ScaleRibbonWidth(Ribbon Width Scale=&FloatFromCurve(
     FloatCurve=DataInterface() {propertyValues {Curve {keys [
         {interpMode=RCIM_Cubic, tangentMode=RCTM_User, time=0, value=0}
         {arriveTangent=0.5028, interpMode=RCIM_Cubic, leaveTangent=0.5028, tangentMode=RCTM_User, time=0.1, value=1.25}
         {arriveTangent=-0.5446, interpMode=RCIM_Cubic, leaveTangent=-0.5446, tangentMode=RCTM_User, time=0.8, value=1}
         {arriveTangent=-3.985, interpMode=RCIM_Cubic, leaveTangent=-3.985, tangentMode=RCTM_User, time=0.9, value=0}
        ]}}}
     Scale Curve=1.25
    ))
   &DynamicMaterialParameters(
    Index 0 Param 1=1
    Index 0 Param 2=0.05
    Module.Index0Param0Write=true
    Module.Index0Param1Write=true
    Module.Index0Param2Write=true
    Module.Index0Param3Write=true
   ) {enabled=false}
   &SolveForcesAndVelocity(Clamp Velocity=false, Limit Acceleration=false)
   &VC_UpdateRibbon_v1(Enable Skeletal Position=true)
  ]
  rendererClasses [&NiagaraRibbonRendererProperties]
  renderers [{rendererClass=&NiagaraRibbonRendererProperties}]
 }
 Ribbon_Electricity002 {
  emitterSpawnScript [/SetVariables(
    Emitter.IsPainted:NiagaraFloat=&MakeCustomFloatFromBool(Bool=$User.bIsPainted, True Float=2, False Float=-0.512)
    Emitter.MirrorOffset:NiagaraFloat=&MakeCustomFloatFromBool(Bool=$User.bIsPainted, True Float=4, False Float=2)
   ) {isSetParameters=true, module=SetVariables_#4}]
  emitterUpdateScript [
   &EmitterState(Life Cycle Mode=::NewEnumerator1, Loop Duration=1, UseLoopDelay=false)
   &VC_FilteredSocketCount()
   &VC_SpawnRibbon_v1(
    Number of Ribbons=$Emitter.PortCount
    SpawnRate=&MakeCustomFloatFromBool(Bool=$User.bIsFrontendPreview, True Float=60, False Float=60)
   )
  ]
  particleSpawnScript [
   &VC_PaintedVariantsOverride(
    Black=<rgba 0.6615 0.6615 0.6615 1>
    BurntSienna=<rgba 0.306 0.1043 0.059 7>
    Cobalt=<rgba 0 0 0.7073 7.6>
    Crimson=<rgba 0.1656 0 0 10>
    DefaultColor=<rgba 18.266 5.7 28.644 1>
    ForestGreen=<rgba 0 1.5 0.0758 10>
    Grey=<rgba 0.274 0.274 0.274 5>
    Lime=<rgba 0.207 1.1 0 6>
    Orange=<rgba 0.6156 0.1614 0 6>
    Pink=<rgba 0.6156 0.0212 0.5289 6>
    Purple=<rgba 0.1963 0 0.8073 7>
    Saffron=<rgba 0.9 0.69 0 10>
    SkyBlue=<rgba 0 0.2944 0.5365 7.5>
    TitaniumWhite=<rgba 0.9448 0.9448 0.9448 7.5>
   )
   &VC_InitializeRibbon_v1(
    Color=$Particles.Color
    Custom Offset=true
    Enable Skeletal Position=true
    Lifetime=&MergeHelper_FloatSwitch(0=0.8, 1=0.8, 2=0.5, 3=?, Index=$User.GameModeID, Num=::NewEnumerator1)
    Local Vector Offset=&AddVectorToPosition(
     Position=&MergeHelper_VectorSwitch(0=<0 0 1>, 1=<0 0 1>, 2=<-58 0 1>, 3=?, Index=$System.GameModeID, Num=::NewEnumerator1)
     Vector=Linked() {linkedVariable {name=User.DTCarOffset003, type=NiagaraPosition}}
    )
    Offset Amount:NiagaraFloat=&MakeFloatFromVector(Vector=<0 0 0>, Channel=::NewEnumerator2)
    Ribbon Width=60
   )
   &AddVelocity(
    Velocity=&MakeCustomVectorFromBool(Bool=$User.bIsFrontendPreview, True Vector=<1000 0 0>, False Vector=<1000 0 0>)
    Velocity Speed Scale=-2
   ) {module=AddVelocity001}
   &ShapeLocation(MaskSpawnGroup=false, Override Seed=false, Sphere Radius=7, UseOwnerScale=true) {enabled=false}
  ]
  particleUpdateScript [
   &ParticleState()
   &InheritSourceMovement(
    Applied Position Scale=&Multiply_VectorByFloat(
     Vector=<1 1 1>
     Float=&FloatFromCurve(FloatCurve=DataInterface() {propertyValues {Curve {keys [{interpMode=RCIM_Cubic, tangentMode=RCTM_Break, time=0, value=0.75}, {interpMode=RCIM_Cubic, tangentMode=RCTM_Break, time=1, value=0}]}}})
    )
    ApplyForce=false
    ApplyPosition=true
    ApplyPositionFalloff=false
   )
   &ScaleColor(
    Scale Alpha=&MakeCustomFloatFromBool(
     Bool=$User.bIsPainted
     True Float=0.15
     False Float=&FloatFromCurve(FloatCurve=DataInterface() {propertyValues {Curve {keys [<at 0 0>,<at 0.15 0.5>,<at 0.7195 0>,<at 1 0>]}}}, Scale Curve=0.2)
    )
    Scale RGB=&MakeCustomVectorFromBool(Bool=$User.bIsPainted, True Vector=<4 4 4>, False Vector=<4 4 4>)
    ScaleA=true
    ScaleRGB=true
   )
   &ScaleRibbonWidth(Ribbon Width Scale=&FloatFromCurve(
     FloatCurve=DataInterface() {propertyValues {Curve {keys [
         {interpMode=RCIM_Cubic, tangentMode=RCTM_User, time=0, value=0}
         {arriveTangent=0.5028, interpMode=RCIM_Cubic, leaveTangent=0.5028, tangentMode=RCTM_User, time=0.1, value=1.25}
         {arriveTangent=-0.5446, interpMode=RCIM_Cubic, leaveTangent=-0.5446, tangentMode=RCTM_User, time=0.8, value=1}
         {arriveTangent=-3.985, interpMode=RCIM_Cubic, leaveTangent=-3.985, tangentMode=RCTM_User, time=0.9, value=0}
        ]}}}
     Scale Curve=1.25
    ))
   &DynamicMaterialParameters(
    Index 0 Param 1=1
    Index 0 Param 2=0.05
    Module.Index0Param0Write=true
    Module.Index0Param1Write=true
    Module.Index0Param2Write=true
    Module.Index0Param3Write=true
   ) {enabled=false}
   &SolveForcesAndVelocity(Clamp Velocity=false, Limit Acceleration=false)
   &VC_UpdateRibbon_v1(Enable Skeletal Position=true)
  ]
  rendererClasses [&NiagaraRibbonRendererProperties]
  renderers [{rendererClass=&NiagaraRibbonRendererProperties}]
 }
 Ribbon_Fire {
  emitterUpdateScript [
   &EmitterState(Inactive Response=::NewEnumerator1, Life Cycle Mode=::NewEnumerator1, Loop Duration=1, UseLoopDelay=false)
   &VC_FilteredSocketCount()
   &VC_RateQualityHandler(Epic Spawn Rate=30, High Spawn Rate=30, Low Spawn Rate=30, Use Custom Rate?=true)
   &VC_SpawnRibbon_v1(Number of Ribbons=$Emitter.PortCount, Spawn Time=-1, SpawnRate=$Emitter.Spawn Rate)
  ]
  particleSpawnScript [
   &VC_PaintedVariantsOverride(
    Black=<rgba 0.01 0.01 0.01 1.25>
    BurntSienna=<rgba 0.0955 0.0295 0.0118 1>
    Cobalt=<rgba 0 0.0066 0.3559 1>
    Crimson=<rgba 0.0816 0 0.0022 1.5>
    ForestGreen=<rgba 0 0.0608 0.0031 1.5>
    Grey=<rgba 0.1649 0.1649 0.1649 0.4435>
    Lime=<rgba 0.0334 0.1565 0.025 1.5>
    Orange=<rgba 0.2205 0.0372 0 1.5>
    Pink=<rgba 0.3003 0.0771 0.2854 1>
    Purple=<rgba 0.0223 0 0.1101 2>
    R1 Black:LinearColor=<rgba 0.0469 0.0469 0.0469 1>
    R1 Cobalt:LinearColor=<rgba 0 0.0185 1 1>
    R1 ForestGreen:LinearColor=<rgba 0 0.2319 0.004 1>
    R1 Grey:LinearColor=<rgba 0.1267 0.1267 0.1267 1>
    R1 Orange:LinearColor=<rgba 0.6724 0.2346 0 1>
    R1 Pink:LinearColor=<rgba 0.5052 0.1552 0.3779 1>
    R1 Purple:LinearColor=<rgba 0.0699 0 0.2552 1>
    R1 SkyBlue:LinearColor=<rgba 0.2058 0.4258 0.5365 1>
    Saffron=<rgba 0.1372 0.1052 0 1>
    SkyBlue=<rgba 0 0.0765 0.1719 1.5>
    TitaniumWhite=<rgba 0.1406 0.1406 0.1406 1>
    Use R1 Painted Overrides=true
   ) {enabled=false}
   &VC_InitializeRibbon_v1(
    Color=$Particles.Color
    Custom Offset=true
    Enable Skeletal Position=true
    Lifetime=&MergeHelper_FloatSwitch(0=0.26, 1=0.255, 2=0.21, 3=?, Index=$User.GameModeID, Num=::NewEnumerator1)
    Local Vector Offset=&AddVectorToPosition(
     Position=&MergeHelper_VectorSwitch(0=<-12.5 0 1>, 1=<0 0 1>, 2=<-17 0 1>, 3=?, Index=$System.GameModeID, Num=::NewEnumerator1)
     Vector=Linked() {linkedVariable {name=User.DTCarOffset003, type=NiagaraPosition}}
    )
    Offset Amount=1
    Ribbon Width=60
    Write Material Random=false
   )
   &AddVelocity(Velocity=<-1750 0 0>) {module=AddVelocity001}
  ]
  particleUpdateScript [
   &ParticleState()
   &InheritSourceMovement(
    Applied Position Scale=&Multiply_VectorByFloat(
     Vector=<1 1 1>
     Float=&FloatFromCurve(FloatCurve=DataInterface() {propertyValues {Curve {keys [
          {interpMode=RCIM_Cubic, tangentMode=RCTM_Break, time=0, value=1}
          {arriveTangent=-0.4819, interpMode=RCIM_Cubic, tangentMode=RCTM_Break, time=1, value=0.75}
         ]}}})
    )
    ApplyForce=false
    ApplyPosition=true
    ApplyPositionFalloff=false
   )
   &SolveForcesAndVelocity(Clamp Velocity=false, Limit Acceleration=false)
   &VC_UpdateRibbon_v1(Enable Skeletal Position=true)
   &VC_PaintedVariantsOverride(
    Black=&ColorFromCurve(ColorCurve=DataInterface() {propertyValues {
       AlphaCurve {keys [{arriveTangent=1.1935, interpMode=RCIM_Cubic, time=0.1355, value=50}]}
       BlueCurve {keys [<at 0.9243 0.0531>]}
       GreenCurve {keys [<at 0.9243 0.0531>]}
       RedCurve {keys [<at 0.9243 0.0531>]}
      }})
    BurntSienna=&ColorFromCurve(ColorCurve=DataInterface() {propertyValues {
       AlphaCurve {keys [{arriveTangent=1.1935, interpMode=RCIM_Cubic, time=0.1355, value=50}]}
       BlueCurve {keys [<at 0.9243 0.036>]}
       GreenCurve {keys [<at 0.9243 0.036>]}
       RedCurve {keys [<at 0.9243 0.1323>]}
      }})
    Cobalt=&ColorFromCurve(ColorCurve=DataInterface() {propertyValues {
       AlphaCurve {keys [{arriveTangent=1.1935, interpMode=RCIM_Cubic, time=0.1355, value=50}]}
       BlueCurve {keys [<at 0.9243 0.2323>]}
       GreenCurve {keys [<at 0.9243 0.0698>]}
       RedCurve {keys [<at 0.9243 0.0448>]}
      }})
    Crimson=&ColorFromCurve(ColorCurve=DataInterface() {propertyValues {
       AlphaCurve {keys [{arriveTangent=1.1935, interpMode=RCIM_Cubic, time=0.1355, value=50}]}
       BlueCurve {keys [<at 0.9243 0>]}
       GreenCurve {keys [<at 0.9243 0>]}
       RedCurve {keys [<at 0.9243 0.3365>]}
      }})
    DefaultColor=&ColorFromCurve(ColorCurve=DataInterface() {propertyValues {
       AlphaCurve {keys [{arriveTangent=1.1935, interpMode=RCIM_Cubic, time=0.1355, value=50}]}
       BlueCurve {keys [<at 0.9243 0.8522>]}
       GreenCurve {keys [<at 0.9243 0.7569>]}
       RedCurve {keys [<at 0.9243 0.2668>]}
      }})
    ForestGreen=&ColorFromCurve(ColorCurve=DataInterface() {propertyValues {
       AlphaCurve {keys [{arriveTangent=1.1935, interpMode=RCIM_Cubic, time=0.1355, value=50}]}
       BlueCurve {keys [<at 0.9243 0.0487>]}
       GreenCurve {keys [<at 0.9243 0.2906>]}
       RedCurve {keys [<at 0.9243 0.0695>]}
      }})
    Grey=&ColorFromCurve(ColorCurve=DataInterface() {propertyValues {
       AlphaCurve {keys [{arriveTangent=1.1935, interpMode=RCIM_Cubic, time=0.1355, value=50}]}
       BlueCurve {keys [<at 0.9243 0.1323>]}
       GreenCurve {keys [<at 0.9243 0.1323>]}
       RedCurve {keys [<at 0.9243 0.1323>]}
      }})
    Lime=&ColorFromCurve(ColorCurve=DataInterface() {propertyValues {
       AlphaCurve {keys [{arriveTangent=1.1935, interpMode=RCIM_Cubic, time=0.1355, value=50}]}
       BlueCurve {keys [<at 0.9243 0.0886>]}
       GreenCurve {keys [<at 0.9243 0.4406>]}
       RedCurve {keys [<at 0.9243 0.2527>]}
      }})
    Orange=&ColorFromCurve(ColorCurve=DataInterface() {propertyValues {
       AlphaCurve {keys [{arriveTangent=1.1935, interpMode=RCIM_Cubic, time=0.1355, value=50}]}
       BlueCurve {keys [<at 0.9243 0.0589>]}
       GreenCurve {keys [<at 0.9243 0.131>]}
       RedCurve {keys [<at 0.9243 0.3698>]}
      }})
    Pink=&ColorFromCurve(ColorCurve=DataInterface() {propertyValues {
       AlphaCurve {keys [{arriveTangent=1.1935, interpMode=RCIM_Cubic, time=0.1355, value=50}]}
       BlueCurve {keys [<at 0.9243 0.2846>]}
       GreenCurve {keys [<at 0.9243 0.1309>]}
       RedCurve {keys [<at 0.9243 0.3198>]}
      }})
    Purple=&ColorFromCurve(ColorCurve=DataInterface() {propertyValues {
       AlphaCurve {keys [{arriveTangent=1.1935, interpMode=RCIM_Cubic, time=0.1355, value=50}]}
       BlueCurve {keys [<at 0.9243 0.3365>]}
       GreenCurve {keys [<at 0.9243 0.0166>]}
       RedCurve {keys [<at 0.9243 0.1192>]}
      }})
    R1 Black:LinearColor=&ColorFromCurve(ColorCurve=DataInterface() {propertyValues {
       AlphaCurve {keys [<at 0.8979 50 RCIM_Cubic>]}
       BlueCurve {keys [<at 0.8183 0>]}
       GreenCurve {keys [<at 0.8183 0>]}
       RedCurve {keys [<at 0.8183 0>]}
      }})
    R1 BurntSienna:LinearColor=&ColorFromCurve(ColorCurve=DataInterface() {propertyValues {
       AlphaCurve {keys [<at 0.8979 50 RCIM_Cubic>]}
       BlueCurve {keys [<at 0.8183 0>]}
       GreenCurve {keys [<at 0.8183 0>]}
       RedCurve {keys [<at 0.8183 0>]}
      }})
    R1 Cobalt:LinearColor=&ColorFromCurve(ColorCurve=DataInterface() {propertyValues {
       AlphaCurve {keys [<at 0.8979 50 RCIM_Cubic>]}
       BlueCurve {keys [<at 0.8183 0>]}
       GreenCurve {keys [<at 0.8183 0>]}
       RedCurve {keys [<at 0.8183 0>]}
      }})
    R1 Crimson:LinearColor=&ColorFromCurve(ColorCurve=DataInterface() {propertyValues {
       AlphaCurve {keys [<at 0.8979 50 RCIM_Cubic>]}
       BlueCurve {keys [<at 0.8183 0>]}
       GreenCurve {keys [<at 0.8183 0>]}
       RedCurve {keys [<at 0.8183 0>]}
      }})
    R1 DefaultColor=&ColorFromCurve(ColorCurve=DataInterface() {propertyValues {
       AlphaCurve {keys [<at 0.8979 50 RCIM_Cubic>]}
       BlueCurve {keys [<at 0.8183 0>]}
       GreenCurve {keys [<at 0.8183 0>]}
       RedCurve {keys [<at 0.8183 0>]}
      }})
    R1 ForestGreen:LinearColor=&ColorFromCurve(ColorCurve=DataInterface() {propertyValues {
       AlphaCurve {keys [<at 0.8979 50 RCIM_Cubic>]}
       BlueCurve {keys [<at 0.8183 0>]}
       GreenCurve {keys [<at 0.8183 0>]}
       RedCurve {keys [<at 0.8183 0>]}
      }})
    R1 Grey:LinearColor=&ColorFromCurve(ColorCurve=DataInterface() {propertyValues {
       AlphaCurve {keys [<at 0.8979 50 RCIM_Cubic>]}
       BlueCurve {keys [<at 0.8183 0>]}
       GreenCurve {keys [<at 0.8183 0>]}
       RedCurve {keys [<at 0.8183 0>]}
      }})
    R1 Lime:LinearColor=&ColorFromCurve(ColorCurve=DataInterface() {propertyValues {
       AlphaCurve {keys [<at 0.8979 50 RCIM_Cubic>]}
       BlueCurve {keys [<at 0.8183 0>]}
       GreenCurve {keys [<at 0.8183 0>]}
       RedCurve {keys [<at 0.8183 0>]}
      }})
    R1 Orange:LinearColor=&ColorFromCurve(ColorCurve=DataInterface() {propertyValues {
       AlphaCurve {keys [<at 0.8979 50 RCIM_Cubic>]}
       BlueCurve {keys [<at 0.8183 0>]}
       GreenCurve {keys [<at 0.8183 0>]}
       RedCurve {keys [<at 0.8183 0>]}
      }})
    R1 Pink:LinearColor=&ColorFromCurve(ColorCurve=DataInterface() {propertyValues {
       AlphaCurve {keys [<at 0.8979 50 RCIM_Cubic>]}
       BlueCurve {keys [<at 0.8183 0>]}
       GreenCurve {keys [<at 0.8183 0>]}
       RedCurve {keys [<at 0.8183 0>]}
      }})
    R1 Purple:LinearColor=&ColorFromCurve(ColorCurve=DataInterface() {propertyValues {
       AlphaCurve {keys [<at 0.8979 50 RCIM_Cubic>]}
       BlueCurve {keys [<at 0.8183 0>]}
       GreenCurve {keys [<at 0.8183 0>]}
       RedCurve {keys [<at 0.8183 0>]}
      }})
    R1 Saffron:LinearColor=&ColorFromCurve(ColorCurve=DataInterface() {propertyValues {
       AlphaCurve {keys [<at 0.8979 50 RCIM_Cubic>]}
       BlueCurve {keys [<at 0.8183 0>]}
       GreenCurve {keys [<at 0.8183 0>]}
       RedCurve {keys [<at 0.8183 0>]}
      }})
    R1 SkyBlue:LinearColor=&ColorFromCurve(ColorCurve=DataInterface() {propertyValues {
       AlphaCurve {keys [<at 0.8979 50 RCIM_Cubic>]}
       BlueCurve {keys [<at 0.8183 0>]}
       GreenCurve {keys [<at 0.8183 0>]}
       RedCurve {keys [<at 0.8183 0>]}
      }})
    R1 TitaniumWhite:LinearColor=&ColorFromCurve(ColorCurve=DataInterface() {propertyValues {
       AlphaCurve {keys [<at 0.8979 50 RCIM_Cubic>]}
       BlueCurve {keys [<at 0.8183 0>]}
       GreenCurve {keys [<at 0.8183 0>]}
       RedCurve {keys [<at 0.8183 0>]}
      }})
    R2 Black:LinearColor=&ColorFromCurve(ColorCurve=DataInterface() {propertyValues {
       AlphaCurve {keys [<at 0.0527 0 RCIM_Cubic>,<at 0.144 1 RCIM_Cubic>,<at 0.8465 1>]}
       BlueCurve {keys [<at 0.6966 0>,<at 0.7247 0>]}
       GreenCurve {keys [<at 0.6966 0>,<at 0.7247 0>]}
       RedCurve {keys [<at 0.6966 0>,<at 0.7247 0>]}
      }})
    R2 BurntSienna:LinearColor=&ColorFromCurve(ColorCurve=DataInterface() {propertyValues {
       AlphaCurve {keys [<at 0.0527 0 RCIM_Cubic>,<at 0.144 1 RCIM_Cubic>,<at 0.8465 1>]}
       BlueCurve {keys [<at 0.6966 0>,<at 0.7247 0>]}
       GreenCurve {keys [<at 0.6966 0>,<at 0.7247 0>]}
       RedCurve {keys [<at 0.6966 0>,<at 0.7247 0>]}
      }})
    R2 Cobalt:LinearColor=&ColorFromCurve(ColorCurve=DataInterface() {propertyValues {
       AlphaCurve {keys [<at 0.0527 0 RCIM_Cubic>,<at 0.144 1 RCIM_Cubic>,<at 0.8465 1>]}
       BlueCurve {keys [<at 0.6966 0>,<at 0.7247 0>]}
       GreenCurve {keys [<at 0.6966 0>,<at 0.7247 0>]}
       RedCurve {keys [<at 0.6966 0>,<at 0.7247 0>]}
      }})
    R2 Crimson:LinearColor=&ColorFromCurve(ColorCurve=DataInterface() {propertyValues {
       AlphaCurve {keys [<at 0.0527 0 RCIM_Cubic>,<at 0.144 1 RCIM_Cubic>,<at 0.8465 1>]}
       BlueCurve {keys [<at 0.6966 0>,<at 0.7247 0>]}
       GreenCurve {keys [<at 0.6966 0>,<at 0.7247 0>]}
       RedCurve {keys [<at 0.6966 0>,<at 0.7247 0>]}
      }})
    R2 DefaultColor:LinearColor=&ColorFromCurve(ColorCurve=DataInterface() {propertyValues {
       AlphaCurve {keys [<at 0.0527 0 RCIM_Cubic>,<at 0.144 1 RCIM_Cubic>,<at 0.8465 1>]}
       BlueCurve {keys [<at 0.6966 0>,<at 0.7247 0>]}
       GreenCurve {keys [<at 0.6966 0>,<at 0.7247 0>]}
       RedCurve {keys [<at 0.6966 0>,<at 0.7247 0>]}
      }})
    R2 ForestGreen:LinearColor=&ColorFromCurve(ColorCurve=DataInterface() {propertyValues {
       AlphaCurve {keys [<at 0.0527 0 RCIM_Cubic>,<at 0.144 1 RCIM_Cubic>,<at 0.8465 1>]}
       BlueCurve {keys [<at 0.6966 0>,<at 0.7247 0>]}
       GreenCurve {keys [<at 0.6966 0>,<at 0.7247 0>]}
       RedCurve {keys [<at 0.6966 0>,<at 0.7247 0>]}
      }})
    R2 Gold:LinearColor=&ColorFromCurve(ColorCurve=DataInterface() {propertyValues {
       AlphaCurve {keys [<at 0.0527 0 RCIM_Cubic>,<at 0.144 1 RCIM_Cubic>,<at 0.8465 1>]}
       BlueCurve {keys [<at 0.6966 0>,<at 0.7247 0>]}
       GreenCurve {keys [<at 0.6966 0>,<at 0.7247 0>]}
       RedCurve {keys [<at 0.6966 0>,<at 0.7247 0>]}
      }})
    R2 Grey:LinearColor=&ColorFromCurve(ColorCurve=DataInterface() {propertyValues {
       AlphaCurve {keys [<at 0.0527 0 RCIM_Cubic>,<at 0.144 1 RCIM_Cubic>,<at 0.8465 1>]}
       BlueCurve {keys [<at 0.6966 0>,<at 0.7247 0>]}
       GreenCurve {keys [<at 0.6966 0>,<at 0.7247 0>]}
       RedCurve {keys [<at 0.6966 0>,<at 0.7247 0>]}
      }})
    R2 Lime:LinearColor=&ColorFromCurve(ColorCurve=DataInterface() {propertyValues {
       AlphaCurve {keys [<at 0.0527 0 RCIM_Cubic>,<at 0.144 1 RCIM_Cubic>,<at 0.8465 1>]}
       BlueCurve {keys [<at 0.6966 0>,<at 0.7247 0>]}
       GreenCurve {keys [<at 0.6966 0>,<at 0.7247 0>]}
       RedCurve {keys [<at 0.6966 0>,<at 0.7247 0>]}
      }})
    R2 Orange:LinearColor=&ColorFromCurve(ColorCurve=DataInterface() {propertyValues {
       AlphaCurve {keys [<at 0.0527 0 RCIM_Cubic>,<at 0.144 1 RCIM_Cubic>,<at 0.8465 1>]}
       BlueCurve {keys [<at 0.6966 0>,<at 0.7247 0>]}
       GreenCurve {keys [<at 0.6966 0>,<at 0.7247 0>]}
       RedCurve {keys [<at 0.6966 0>,<at 0.7247 0>]}
      }})
    R2 Pink:LinearColor=&ColorFromCurve(ColorCurve=DataInterface() {propertyValues {
       AlphaCurve {keys [<at 0.0527 0 RCIM_Cubic>,<at 0.144 1 RCIM_Cubic>,<at 0.8465 1>]}
       BlueCurve {keys [<at 0.6966 0>,<at 0.7247 0>]}
       GreenCurve {keys [<at 0.6966 0>,<at 0.7247 0>]}
       RedCurve {keys [<at 0.6966 0>,<at 0.7247 0>]}
      }})
    R2 Purple:LinearColor=&ColorFromCurve(ColorCurve=DataInterface() {propertyValues {
       AlphaCurve {keys [<at 0.0527 0 RCIM_Cubic>,<at 0.144 1 RCIM_Cubic>,<at 0.8465 1>]}
       BlueCurve {keys [<at 0.6966 0>,<at 0.7247 0>]}
       GreenCurve {keys [<at 0.6966 0>,<at 0.7247 0>]}
       RedCurve {keys [<at 0.6966 0>,<at 0.7247 0>]}
      }})
    R2 Saffron:LinearColor=&ColorFromCurve(ColorCurve=DataInterface() {propertyValues {
       AlphaCurve {keys [<at 0.0527 0 RCIM_Cubic>,<at 0.144 1 RCIM_Cubic>,<at 0.8465 1>]}
       BlueCurve {keys [<at 0.6966 0>,<at 0.7247 0>]}
       GreenCurve {keys [<at 0.6966 0>,<at 0.7247 0>]}
       RedCurve {keys [<at 0.6966 0>,<at 0.7247 0>]}
      }})
    R2 SkyBlue:LinearColor=&ColorFromCurve(ColorCurve=DataInterface() {propertyValues {
       AlphaCurve {keys [<at 0.0527 0 RCIM_Cubic>,<at 0.144 1 RCIM_Cubic>,<at 0.8465 1>]}
       BlueCurve {keys [<at 0.6966 0>,<at 0.7247 0>]}
       GreenCurve {keys [<at 0.6966 0>,<at 0.7247 0>]}
       RedCurve {keys [<at 0.6966 0>,<at 0.7247 0>]}
      }})
    R2 TitaniumWhite:LinearColor=&ColorFromCurve(ColorCurve=DataInterface() {propertyValues {
       AlphaCurve {keys [<at 0.0527 0 RCIM_Cubic>,<at 0.144 1 RCIM_Cubic>,<at 0.8465 1>]}
       BlueCurve {keys [<at 0.6966 0>,<at 0.7247 0>]}
       GreenCurve {keys [<at 0.6966 0>,<at 0.7247 0>]}
       RedCurve {keys [<at 0.6966 0>,<at 0.7247 0>]}
      }})
    Saffron=&ColorFromCurve(ColorCurve=DataInterface() {propertyValues {
       AlphaCurve {keys [{arriveTangent=1.1935, interpMode=RCIM_Cubic, time=0.1355, value=50}]}
       BlueCurve {keys [<at 0.9243 0.0211>]}
       GreenCurve {keys [<at 0.9243 0.4248>]}
       RedCurve {keys [<at 0.9243 0.4448>]}
      }})
    SkyBlue=&ColorFromCurve(ColorCurve=DataInterface() {propertyValues {
       AlphaCurve {keys [{arriveTangent=1.1935, interpMode=RCIM_Cubic, time=0.1355, value=50}]}
       BlueCurve {keys [<at 0.9243 0.349>]}
       GreenCurve {keys [<at 0.9243 0.198>]}
       RedCurve {keys [<at 0.9243 0.0173>]}
      }})
    TitaniumWhite=&ColorFromCurve(ColorCurve=DataInterface() {propertyValues {
       AlphaCurve {keys [{arriveTangent=1.1935, interpMode=RCIM_Cubic, time=0.1355, value=50}]}
       BlueCurve {keys [<at 0.9243 1>]}
       GreenCurve {keys [<at 0.9243 1>]}
       RedCurve {keys [<at 0.9243 1>]}
      }})
    Use R1 Painted Overrides=true
    Use R2 Painted Overrides=true
   ) {module=VC_PaintedVariantsOverride002}
   /SetVariables(
    Particles.RibbonBase Color:LinearColor=&ColorFromCurve(ColorCurve=DataInterface() {propertyValues {
       AlphaCurve {keys [<at 0.144 50 RCIM_Cubic>,<at 0.8465 50>]}
       BlueCurve {keys [<at 0.6966 0>,<at 0.7247 0>]}
       GreenCurve {keys [<at 0.6966 0>,<at 0.7247 0>]}
       RedCurve {keys [<at 0.6966 0>,<at 0.7247 0>]}
      }})
    Particles.RibbonBase2 Color:LinearColor=&ColorFromCurve(ColorCurve=DataInterface() {propertyValues {
       AlphaCurve {keys [{arriveTangent=1.1935, interpMode=RCIM_Cubic, time=0.1355, value=50}]}
       BlueCurve {keys [<at 0.9243 0.8522>]}
       GreenCurve {keys [<at 0.9243 0.7569>]}
       RedCurve {keys [<at 0.9243 0.2668>]}
      }})
    Particles.curve_color_lines:LinearColor=&MultiplyLinearColor(
     Linear Color=$Particles.color_lines
     Scale Factor=&MakeVector4_FromLinearColor(LinearColor=&MakeLinearColorFromVectorAndFloat(
       "Vector (RGB)"=<1 1 1>
       "Float (Alpha)"=&FloatFromCurve(FloatCurve=DataInterface() {propertyValues {Curve {keys [{interpMode=RCIM_Cubic, tangentMode=RCTM_Break, time=0.5, value=1}, {interpMode=RCIM_Cubic, tangentMode=RCTM_Break, time=1, value=1}]}}})
      ))
    )
   ) {enabled=false, isSetParameters=true, module=SetVariables_#5}
   &Color(Color=&MakeCustomLinearColorFromBool(
     Bool=$User.bIsPainted
     True Color=$Particles.Color
     False Color=&ColorFromCurve(ColorCurve=DataInterface() {propertyValues {
        bExposeCurve=true
        AlphaCurve {keys [<at 0.0095 0>,<at 0.0365 1>,<at 1 1 RCIM_Cubic>]}
        BlueCurve {defaultValue=0, keys [<at 0.8537 1>]}
        GreenCurve {defaultValue=0, keys [<at 0.8537 0.7702>]}
        RedCurve {defaultValue=1, keys [<at 0.8537 0.5989>]}
       }})
    )) {enabled=false}
   &ScaleRibbonWidth(Ribbon Width Scale=&FloatFromCurve(
     FloatCurve=DataInterface() {propertyValues {Curve {keys [
         {arriveTangent=0.3269, interpMode=RCIM_Cubic, time=0, value=1.1}
         {arriveTangent=-0.9052, interpMode=RCIM_Cubic, leaveTangent=-0.9052, time=0.5389, value=0.8752}
         {arriveTangent=-1.8925, arriveTangentWeight=1.0346, tangentMode=RCTM_User, time=0.8188, value=0.3588}
         <at 1.002 0.3176>
        ]}}}
     Scale Curve=1.2
    ))
   /SetVariables(Particles.WPO:NiagaraPosition=&AddVectorToPosition(
     Position=$Particles.Position
     Vector=&TransformVector(
      Vector=&MakeVector(
       X=0
       Y=0
       Z=&Sine(
        Normalized Angle=&Add_Float(
         A=&Multiply_Float(A=$Particles.NormalizedAge, B=-8)
         B=&Multiply_Float(A=$Emitter.LoopedAge, B=6)
        )
        Period=&MaskFloatBySpawnGroup(Masked Float=-0.85, Passthrough Float=0.85)
        Scale=&FloatFromCurve(
         FloatCurve=DataInterface() {propertyValues {Curve {
            defaultValue=1
            keys [
             {interpMode=RCIM_Cubic, tangentMode=RCTM_User, time=0, value=0}
             {arriveTangent=1.3728, interpMode=RCIM_Cubic, leaveTangent=1.3728, tangentMode=RCTM_User, time=0.5423, value=0.2659}
             {interpMode=RCIM_Cubic, tangentMode=RCTM_User, time=1, value=1.5}
            ]
           }}}
         Scale Curve=10
        )
        Bias=0
       )
      )
      Destination Space=::Simulation
      Apply Scale=true
     )
    )) {isSetParameters=true, module=SetVariables_#6}
  ]
  rendererClasses [&NiagaraRibbonRendererProperties]
  renderers [
   {rendererClass=&NiagaraRibbonRendererProperties}
   {rendererClass=&NiagaraRibbonRendererProperties, rendererIndex=1}
   {rendererClass=&NiagaraRibbonRendererProperties, rendererIndex=2}
  ]
 }
 Ribbon_Fire001 {
  bEnabled=false
  emitterUpdateScript [
   &EmitterState(Inactive Response=::NewEnumerator1, Life Cycle Mode=::NewEnumerator1, Loop Duration=1, UseLoopDelay=false)
   &VC_FilteredSocketCount()
   &VC_RateQualityHandler(Epic Spawn Rate=30, High Spawn Rate=30, Low Spawn Rate=30, Use Custom Rate?=true)
   &VC_SpawnRibbon_v1(Number of Ribbons=$Emitter.PortCount, Spawn Time=-1, SpawnRate=$Emitter.Spawn Rate)
  ]
  particleSpawnScript [
   &VC_PaintedVariantsOverride(
    Black=<rgba 0.01 0.01 0.01 1.25>
    BurntSienna=<rgba 0.0955 0.0295 0.0118 1>
    Cobalt=<rgba 0 0.0066 0.3559 1>
    Crimson=<rgba 0.0816 0 0.0022 1.5>
    ForestGreen=<rgba 0 0.0608 0.0031 1.5>
    Grey=<rgba 0.1649 0.1649 0.1649 0.4435>
    Lime=<rgba 0.0334 0.1565 0.025 1.5>
    Orange=<rgba 0.2205 0.0372 0 1.5>
    Pink=<rgba 0.3003 0.0771 0.2854 1>
    Purple=<rgba 0.0223 0 0.1101 2>
    R1 Black:LinearColor=<rgba 0.0469 0.0469 0.0469 1>
    R1 Cobalt:LinearColor=<rgba 0 0.0185 1 1>
    R1 ForestGreen:LinearColor=<rgba 0 0.2319 0.004 1>
    R1 Grey:LinearColor=<rgba 0.1267 0.1267 0.1267 1>
    R1 Orange:LinearColor=<rgba 0.6724 0.2346 0 1>
    R1 Pink:LinearColor=<rgba 0.5052 0.1552 0.3779 1>
    R1 Purple:LinearColor=<rgba 0.0699 0 0.2552 1>
    R1 SkyBlue:LinearColor=<rgba 0.2058 0.4258 0.5365 1>
    Saffron=<rgba 0.1372 0.1052 0 1>
    SkyBlue=<rgba 0 0.0765 0.1719 1.5>
    TitaniumWhite=<rgba 0.1406 0.1406 0.1406 1>
    Use R1 Painted Overrides=true
   )
   &VC_InitializeRibbon_v1(
    Color=$Particles.Color
    Custom Offset=true
    Enable Skeletal Position=true
    Lifetime=&MergeHelper_FloatSwitch(0=0.26, 1=0.255, 2=0.21, 3=?, Index=$User.GameModeID, Num=::NewEnumerator1)
    Local Vector Offset=&AddVectorToPosition(
     Position=&MergeHelper_VectorSwitch(0=<-12.5 0 1>, 1=<0 0 1>, 2=<-17 0 1>, 3=?, Index=$System.GameModeID, Num=::NewEnumerator1)
     Vector=Linked() {linkedVariable {name=User.DTCarOffset003, type=NiagaraPosition}}
    )
    Offset Amount=1
    Ribbon Width=60
    Write Material Random=false
   )
   &AddVelocity(Velocity=<-1750 0 0>) {module=AddVelocity001}
  ]
  particleUpdateScript [
   &ParticleState()
   &InheritSourceMovement(
    Applied Position Scale=&Multiply_VectorByFloat(
     Vector=<1 1 1>
     Float=&FloatFromCurve(FloatCurve=DataInterface() {propertyValues {Curve {keys [
          {interpMode=RCIM_Cubic, tangentMode=RCTM_Break, time=0, value=1}
          {arriveTangent=-0.4819, interpMode=RCIM_Cubic, tangentMode=RCTM_Break, time=1, value=0.75}
         ]}}})
    )
    ApplyForce=false
    ApplyPosition=true
    ApplyPositionFalloff=false
   )
   &SolveForcesAndVelocity(Clamp Velocity=false, Limit Acceleration=false)
   &VC_UpdateRibbon_v1(Enable Skeletal Position=true)
   /SetVariables(
    Particles.RibbonBase Color:LinearColor=&ColorFromCurve(ColorCurve=DataInterface() {propertyValues {
       AlphaCurve {keys [<at 0.144 50 RCIM_Cubic>,<at 0.8465 50>]}
       BlueCurve {keys [<at 0.6966 0>,<at 0.7247 0>]}
       GreenCurve {keys [<at 0.6966 0>,<at 0.7247 0>]}
       RedCurve {keys [<at 0.6966 0>,<at 0.7247 0>]}
      }})
    Particles.RibbonBase2 Color:LinearColor=&ColorFromCurve(ColorCurve=DataInterface() {propertyValues {
       AlphaCurve {keys [{arriveTangent=1.1935, interpMode=RCIM_Cubic, time=0.1355, value=50}]}
       BlueCurve {keys [<at 0.9243 0.8522>]}
       GreenCurve {keys [<at 0.9243 0.7569>]}
       RedCurve {keys [<at 0.9243 0.2668>]}
      }})
    Particles.curve_color_lines:LinearColor=&MultiplyLinearColor(
     Linear Color=$Particles.color_lines
     Scale Factor=&MakeVector4_FromLinearColor(LinearColor=&MakeLinearColorFromVectorAndFloat(
       "Vector (RGB)"=<1 1 1>
       "Float (Alpha)"=&FloatFromCurve(FloatCurve=DataInterface() {propertyValues {Curve {keys [{interpMode=RCIM_Cubic, tangentMode=RCTM_Break, time=0.5, value=1}, {interpMode=RCIM_Cubic, tangentMode=RCTM_Break, time=1, value=1}]}}})
      ))
    )
   ) {isSetParameters=true, module=SetVariables_#5}
   &Color(Color=&MakeCustomLinearColorFromBool(
     Bool=$User.bIsPainted
     True Color=$Particles.Color
     False Color=&ColorFromCurve(ColorCurve=DataInterface() {propertyValues {
        bExposeCurve=true
        AlphaCurve {keys [<at 0.0095 0>,<at 0.0365 1>,<at 1 1 RCIM_Cubic>]}
        BlueCurve {defaultValue=0, keys [<at 0.8537 1>]}
        GreenCurve {defaultValue=0, keys [<at 0.8537 0.7702>]}
        RedCurve {defaultValue=1, keys [<at 0.8537 0.5989>]}
       }})
    )) {enabled=false}
   &ScaleRibbonWidth(Ribbon Width Scale=&FloatFromCurve(
     FloatCurve=DataInterface() {propertyValues {Curve {keys [
         {arriveTangent=0.3269, interpMode=RCIM_Cubic, time=0, value=1.1}
         {arriveTangent=-0.9052, interpMode=RCIM_Cubic, leaveTangent=-0.9052, time=0.5389, value=0.8752}
         {arriveTangent=-1.8925, arriveTangentWeight=1.0346, tangentMode=RCTM_User, time=0.8188, value=0.3588}
         <at 1.002 0.3176>
        ]}}}
     Scale Curve=1.2
    ))
   /SetVariables(Particles.WPO:NiagaraPosition=&AddVectorToPosition(
     Position=$Particles.Position
     Vector=&TransformVector(
      Vector=&MakeVector(
       X=0
       Y=0
       Z=&Sine(
        Normalized Angle=&Add_Float(
         A=&Multiply_Float(A=$Particles.NormalizedAge, B=-8)
         B=&Multiply_Float(A=$Emitter.LoopedAge, B=6)
        )
        Period=&MaskFloatBySpawnGroup(Masked Float=-0.85, Passthrough Float=0.85)
        Scale=&FloatFromCurve(
         FloatCurve=DataInterface() {propertyValues {Curve {
            defaultValue=1
            keys [
             {interpMode=RCIM_Cubic, tangentMode=RCTM_User, time=0, value=0}
             {arriveTangent=1.3728, interpMode=RCIM_Cubic, leaveTangent=1.3728, tangentMode=RCTM_User, time=0.5423, value=0.2659}
             {interpMode=RCIM_Cubic, tangentMode=RCTM_User, time=1, value=1.5}
            ]
           }}}
         Scale Curve=10
        )
        Bias=0
       )
      )
      Destination Space=::Simulation
      Apply Scale=true
     )
    )) {isSetParameters=true, module=SetVariables_#6}
  ]
  rendererClasses [&NiagaraRibbonRendererProperties]
  renderers [
   {rendererClass=&NiagaraRibbonRendererProperties}
   {rendererClass=&NiagaraRibbonRendererProperties, rendererIndex=1}
   {rendererClass=&NiagaraRibbonRendererProperties, rendererIndex=2}
  ]
 }
 Ribbon_Trail {
  emitterUpdateScript [
   &EmitterState(Inactive Response=::NewEnumerator1, Life Cycle Mode=::NewEnumerator1, Loop Duration=1, UseLoopDelay=false)
   &VC_FilteredSocketCount()
   &VC_RateQualityHandler(Epic Spawn Rate=30, High Spawn Rate=30, Low Spawn Rate=30, Use Custom Rate?=true)
   &VC_SpawnRibbon_v1(Number of Ribbons=$Emitter.PortCount, Spawn Time=-1, SpawnRate=$Emitter.Spawn Rate)
  ]
  particleSpawnScript [
   &VC_PaintedVariantsOverride(
    Black=<rgba 0.01 0.01 0.01 1.25>
    BurntSienna=<rgba 0.0955 0.0295 0.0118 1>
    Cobalt=<rgba 0 0.0066 0.3559 1>
    Crimson=<rgba 0.0816 0 0.0022 1.5>
    ForestGreen=<rgba 0 0.0608 0.0031 1.5>
    Grey=<rgba 0.1649 0.1649 0.1649 0.4435>
    Lime=<rgba 0.0334 0.1565 0.025 1.5>
    Orange=<rgba 0.2205 0.0372 0 1.5>
    Pink=<rgba 0.3003 0.0771 0.2854 1>
    Purple=<rgba 0.0223 0 0.1101 2>
    R1 Black:LinearColor=<rgba 0.0469 0.0469 0.0469 1>
    R1 Cobalt:LinearColor=<rgba 0 0.0185 1 1>
    R1 ForestGreen:LinearColor=<rgba 0 0.2319 0.004 1>
    R1 Grey:LinearColor=<rgba 0.1267 0.1267 0.1267 1>
    R1 Orange:LinearColor=<rgba 0.6724 0.2346 0 1>
    R1 Pink:LinearColor=<rgba 0.5052 0.1552 0.3779 1>
    R1 Purple:LinearColor=<rgba 0.0699 0 0.2552 1>
    R1 SkyBlue:LinearColor=<rgba 0.2058 0.4258 0.5365 1>
    Saffron=<rgba 0.1372 0.1052 0 1>
    SkyBlue=<rgba 0 0.0765 0.1719 1.5>
    TitaniumWhite=<rgba 0.1406 0.1406 0.1406 1>
    Use R1 Painted Overrides=true
   ) {enabled=false}
   &VC_InitializeRibbon_v1(
    Color=$Particles.Color
    Custom Offset=true
    Enable Skeletal Position=true
    Lifetime=&MergeHelper_FloatSwitch(0=1, 1=0.8, 2=0.5, 3=?, Index=$User.GameModeID, Num=::NewEnumerator1)
    Local Vector Offset=&AddVectorToPosition(
     Position=&MergeHelper_VectorSwitch(0=<-3.82 0 0>, 1=<-3.821 0 0>, 2=<-55 0 0>, 3=?, Index=$System.GameModeID, Num=::NewEnumerator1)
     Vector=Linked() {linkedVariable {name=User.DTCarOffset003, type=NiagaraPosition}}
    )
    Offset Amount=1
    Ribbon Width=42
   )
   &AddVelocity(Velocity=<-1500 0 0>) {module=AddVelocity001}
  ]
  particleUpdateScript [
   &ParticleState()
   &InheritSourceMovement(
    Applied Position Scale=&Multiply_VectorByFloat(
     Vector=<1 1 1>
     Float=&FloatFromCurve(FloatCurve=DataInterface() {propertyValues {Curve {keys [{interpMode=RCIM_Cubic, tangentMode=RCTM_Break, time=0, value=0.85}, {interpMode=RCIM_Cubic, tangentMode=RCTM_Break, time=1, value=0}]}}})
    )
    ApplyForce=false
    ApplyPosition=true
    ApplyPositionFalloff=false
   )
   &SolveForcesAndVelocity(Clamp Velocity=false, Limit Acceleration=false)
   /SetVariables(
    Particles.Ribbon 2 Base:LinearColor=&ColorFromCurve(ColorCurve=DataInterface() {propertyValues {
       AlphaCurve {keys [<at 0.8979 1 RCIM_Cubic>,<at 0.9431 0 RCIM_Cubic>]}
       BlueCurve {keys [<at 0.6799 0.9 RCIM_Cubic>,<at 0.8183 0>]}
       GreenCurve {keys [<at 0.6799 0.8678 RCIM_Cubic>,<at 0.8183 0>]}
       RedCurve {keys [<at 0.6799 0.7904 RCIM_Cubic>,<at 0.8183 0>]}
      }})
    Particles.RibbonBase Color:LinearColor=&ColorFromCurve(ColorCurve=DataInterface() {propertyValues {
       AlphaCurve {keys [<at 0.0527 0 RCIM_Cubic>,<at 0.144 1 RCIM_Cubic>,<at 0.8465 1>]}
       BlueCurve {keys [<at 0.6966 0>,<at 0.7247 0>]}
       GreenCurve {keys [<at 0.6966 0>,<at 0.7247 0>]}
       RedCurve {keys [<at 0.6966 0>,<at 0.7247 0>]}
      }})
    Particles.RibbonBase2 Color:LinearColor=&ColorFromCurve(ColorCurve=DataInterface() {propertyValues {
       AlphaCurve {keys [<at 0.0038 0 RCIM_Cubic>,<at 0.0364 1 RCIM_Cubic>,<at 0.8276 1>,<at 0.8502 0>]}
       BlueCurve {keys [<at 0.6367 0>,<at 0.6543 0>]}
       GreenCurve {keys [<at 0.6367 0>,<at 0.6543 0>]}
       RedCurve {keys [<at 0.6367 1.25>,<at 0.6543 0>]}
      }})
    Particles.curve_color_lines:LinearColor=&MultiplyLinearColor(
     Linear Color=$Particles.color_lines
     Scale Factor=&MakeVector4_FromLinearColor(LinearColor=&MakeLinearColorFromVectorAndFloat(
       "Vector (RGB)"=<1 1 1>
       "Float (Alpha)"=&FloatFromCurve(FloatCurve=DataInterface() {propertyValues {Curve {keys [{interpMode=RCIM_Cubic, tangentMode=RCTM_Break, time=0.5, value=1}, {interpMode=RCIM_Cubic, tangentMode=RCTM_Break, time=1, value=1}]}}})
      ))
    )
   ) {enabled=false, isSetParameters=true, module=SetVariables_#7}
   &VC_UpdateRibbon_v1(Enable Skeletal Position=true)
   &Color(Color=&MakeCustomLinearColorFromBool(
     Bool=$User.bIsPainted
     True Color=$Particles.Color
     False Color=&ColorFromCurve(ColorCurve=DataInterface() {propertyValues {
        bExposeCurve=true
        AlphaCurve {keys [<at 0.0095 0>,<at 0.0365 1>,<at 1 1 RCIM_Cubic>]}
        BlueCurve {defaultValue=0, keys [<at 0.8537 1>]}
        GreenCurve {defaultValue=0, keys [<at 0.8537 0.7702>]}
        RedCurve {defaultValue=1, keys [<at 0.8537 0.5989>]}
       }})
    )) {enabled=false}
   &ScaleRibbonWidth(Ribbon Width Scale=&FloatFromCurve(
     FloatCurve=DataInterface() {propertyValues {Curve {keys [
         {arriveTangent=-0.5896, interpMode=RCIM_Cubic, leaveTangent=-0.5896, tangentMode=RCTM_User, time=0.0029, value=0.5874}
         {arriveTangent=1.9541, interpMode=RCIM_Cubic, leaveTangent=1.9541, tangentMode=RCTM_User, time=0.0963, value=0.8133}
         {arriveTangent=2.1445, interpMode=RCIM_Cubic, leaveTangent=2.1445, tangentMode=RCTM_User, time=0.209, value=0.8907}
         {arriveTangent=-0.0014, interpMode=RCIM_Cubic, leaveTangent=-0.0014, tangentMode=RCTM_User, time=0.4609, value=1.0063}
         {
          arriveTangent=-0.5896
          arriveTangentWeight=1.0346
          interpMode=RCIM_Cubic
          leaveTangent=-0.5896
          tangentMode=RCTM_User
          time=0.8245
          value=0.4328
         }
         {arriveTangent=-0.5896, interpMode=RCIM_Cubic, leaveTangent=-0.5896, tangentMode=RCTM_User, time=1.002, value=0.3176}
        ]}}}
     Scale Curve=1.2
    ))
   /SetVariables(Particles.BlueRibbon_Width:NiagaraFloat=&FloatFromCurve(FloatCurve=DataInterface() {propertyValues {Curve {keys [
         {arriveTangent=0.0034, interpMode=RCIM_Cubic, leaveTangent=0.0034, tangentMode=RCTM_User, time=-0.0022, value=42}
         {arriveTangent=-20.8009, interpMode=RCIM_Cubic, leaveTangent=-20.8009, tangentMode=RCTM_User, time=0.5469, value=36.2918}
         {arriveTangent=-0.0044, interpMode=RCIM_Cubic, leaveTangent=-0.0044, tangentMode=RCTM_User, time=1.007, value=15.4938}
        ]}}})) {isSetParameters=true, module=SetVariables_#8}
   /SetVariables(Particles.WPO:NiagaraPosition=&AddVectorToPosition(
     Position=$Particles.Position
     Vector=&TransformVector(
      Vector=&MakeVector(
       X=0
       Y=0
       Z=&Sine(
        Normalized Angle=&Add_Float(
         A=&Multiply_Float(A=$Particles.NormalizedAge, B=-1)
         B=&Multiply_Float(A=$Emitter.LoopedAge, B=4)
        )
        Period=&MaskFloatBySpawnGroup(Masked Float=0.5, Passthrough Float=-0.5)
        Scale=&FloatFromCurve(
         FloatCurve=DataInterface() {propertyValues {Curve {
            defaultValue=1
            keys [
             {interpMode=RCIM_Cubic, tangentMode=RCTM_User, time=0, value=0}
             {arriveTangent=0.4319, interpMode=RCIM_Cubic, leaveTangent=0.4319, time=0.489, value=0.2385}
             {interpMode=RCIM_Cubic, tangentMode=RCTM_User, time=1.0038, value=0.4336}
            ]
           }}}
         Scale Curve=10
        )
        Bias=0
       )
      )
      Destination Space=::Simulation
      Apply Scale=true
     )
    )) {isSetParameters=true, module=SetVariables_#9}
   &DynamicMaterialParameters(
    Index 0 Param 4=10
    Module.Index0Float0Write=true
    Module.Index0Float1Write=true
    Module.Index0Float2Write=true
    Module.Index0Float3Write=true
   )
   &VC_PaintedVariantsOverride(
    Black=&ColorFromCurve(ColorCurve=DataInterface() {propertyValues {
       AlphaCurve {keys [<at 0.0038 0 RCIM_Cubic>,<at 0.0364 1 RCIM_Cubic>,<at 0.8276 1>,<at 0.8502 0>]}
       BlueCurve {keys [<at 0.6367 0.1948>,<at 0.6543 0>]}
       GreenCurve {keys [<at 0.6367 0.1948>,<at 0.6543 0>]}
       RedCurve {keys [<at 0.6367 0.1948>,<at 0.6543 0>]}
      }})
    BurntSienna=&ColorFromCurve(ColorCurve=DataInterface() {propertyValues {
       AlphaCurve {keys [<at 0.0038 0 RCIM_Cubic>,<at 0.0364 1 RCIM_Cubic>,<at 0.8276 1>,<at 0.8502 0>]}
       BlueCurve {keys [<at 0.6367 0.0735>,<at 0.6543 0>]}
       GreenCurve {keys [<at 0.6367 0.0955>,<at 0.6543 0>]}
       RedCurve {keys [<at 0.6367 0.2226>,<at 0.6543 0>]}
      }})
    Cobalt=&ColorFromCurve(ColorCurve=DataInterface() {propertyValues {
       AlphaCurve {keys [<at 0.0038 0 RCIM_Cubic>,<at 0.0364 1 RCIM_Cubic>,<at 0.8276 1>,<at 0.8502 0>]}
       BlueCurve {keys [<at 0.6367 0.6914>,<at 0.6543 0>]}
       GreenCurve {keys [<at 0.6367 0.1366>,<at 0.6543 0>]}
       RedCurve {keys [<at 0.6367 0.0727>,<at 0.6543 0>]}
      }})
    Crimson=&ColorFromCurve(ColorCurve=DataInterface() {propertyValues {
       AlphaCurve {keys [<at 0.0038 0 RCIM_Cubic>,<at 0.0364 1 RCIM_Cubic>,<at 0.8276 1>,<at 0.8502 0>]}
       BlueCurve {keys [<at 0.6367 0.0314>,<at 0.6543 0>]}
       GreenCurve {keys [<at 0.6367 0>,<at 0.6543 0>]}
       RedCurve {keys [<at 0.6367 0.4031>,<at 0.6543 0>]}
      }})
    DefaultColor=&ColorFromCurve(ColorCurve=DataInterface() {propertyValues {
       AlphaCurve {keys [<at 0.0038 0 RCIM_Cubic>,<at 0.0364 1 RCIM_Cubic>,<at 0.8276 1>,<at 0.8502 0>]}
       BlueCurve {keys [<at 0.6367 0>,<at 0.6543 0>]}
       GreenCurve {keys [<at 0.6367 0>,<at 0.6543 0>]}
       RedCurve {keys [<at 0.6367 1.25>,<at 0.6543 0>]}
      }})
    ForestGreen=&ColorFromCurve(ColorCurve=DataInterface() {propertyValues {
       AlphaCurve {keys [<at 0.0038 0 RCIM_Cubic>,<at 0.0364 1 RCIM_Cubic>,<at 0.8276 1>,<at 0.8502 0>]}
       BlueCurve {keys [<at 0.6367 0.1571>,<at 0.6543 0>]}
       GreenCurve {keys [<at 0.6367 0.6948>,<at 0.6543 0>]}
       RedCurve {keys [<at 0.6367 0.2883>,<at 0.6543 0>]}
      }})
    Grey=&ColorFromCurve(ColorCurve=DataInterface() {propertyValues {
       AlphaCurve {keys [<at 0.0038 0 RCIM_Cubic>,<at 0.0364 1 RCIM_Cubic>,<at 0.8276 1>,<at 0.8502 0>]}
       BlueCurve {keys [<at 0.6367 0.2823>,<at 0.6543 0>]}
       GreenCurve {keys [<at 0.6367 0.2823>,<at 0.6543 0>]}
       RedCurve {keys [<at 0.6367 0.2823>,<at 0.6543 0>]}
      }})
    Lime=&ColorFromCurve(ColorCurve=DataInterface() {propertyValues {
       AlphaCurve {keys [<at 0.0038 0 RCIM_Cubic>,<at 0.0364 1 RCIM_Cubic>,<at 0.8276 1>,<at 0.8502 0>]}
       BlueCurve {keys [<at 0.6367 0.0737>,<at 0.6543 0>]}
       GreenCurve {keys [<at 0.6367 0.9674>,<at 0.6543 0>]}
       RedCurve {keys [<at 0.6367 0.5742>,<at 0.6543 0>]}
      }})
    Orange=&ColorFromCurve(ColorCurve=DataInterface() {propertyValues {
       AlphaCurve {keys [<at 0.0038 0 RCIM_Cubic>,<at 0.0364 1 RCIM_Cubic>,<at 0.8276 1>,<at 0.8502 0>]}
       BlueCurve {keys [<at 0.6367 0.128>,<at 0.6543 0>]}
       GreenCurve {keys [<at 0.6367 0.3757>,<at 0.6543 0>]}
       RedCurve {keys [<at 0.6367 1.1706>,<at 0.6543 0>]}
      }})
    Pink=&ColorFromCurve(ColorCurve=DataInterface() {propertyValues {
       AlphaCurve {keys [<at 0.0038 0 RCIM_Cubic>,<at 0.0364 1 RCIM_Cubic>,<at 0.8276 1>,<at 0.8502 0>]}
       BlueCurve {keys [<at 0.6367 0.6974>,<at 0.6543 0>]}
       GreenCurve {keys [<at 0.6367 0.3161>,<at 0.6543 0>]}
       RedCurve {keys [<at 0.6367 0.7983>,<at 0.6543 0>]}
      }})
    Purple=&ColorFromCurve(ColorCurve=DataInterface() {propertyValues {
       AlphaCurve {keys [<at 0.0038 0 RCIM_Cubic>,<at 0.0364 1 RCIM_Cubic>,<at 0.8276 1>,<at 0.8502 0>]}
       BlueCurve {keys [<at 0.6367 0.4518>,<at 0.6543 0>]}
       GreenCurve {keys [<at 0.6367 0.0452>,<at 0.6543 0>]}
       RedCurve {keys [<at 0.6367 0.1674>,<at 0.6543 0>]}
      }})
    R1 Black:LinearColor=&ColorFromCurve(ColorCurve=DataInterface() {propertyValues {
       AlphaCurve {keys [<at 0.8979 1 RCIM_Cubic>,<at 0.9431 0 RCIM_Cubic>]}
       BlueCurve {keys [<at 0.6799 0.1573 RCIM_Cubic>,<at 0.8183 0>]}
       GreenCurve {keys [<at 0.6799 0.1573 RCIM_Cubic>,<at 0.8183 0>]}
       RedCurve {keys [<at 0.6799 0.1573 RCIM_Cubic>,<at 0.8183 0>]}
      }})
    R1 BurntSienna:LinearColor=&ColorFromCurve(ColorCurve=DataInterface() {propertyValues {
       AlphaCurve {keys [<at 0.8979 1 RCIM_Cubic>,<at 0.9431 0 RCIM_Cubic>]}
       BlueCurve {keys [<at 0.6799 0.1415 RCIM_Cubic>,<at 0.8183 0>]}
       GreenCurve {keys [<at 0.6799 0.1479 RCIM_Cubic>,<at 0.8183 0>]}
       RedCurve {keys [<at 0.6799 0.3026 RCIM_Cubic>,<at 0.8183 0>]}
      }})
    R1 Cobalt:LinearColor=&ColorFromCurve(ColorCurve=DataInterface() {propertyValues {
       AlphaCurve {keys [<at 0.8979 1 RCIM_Cubic>,<at 0.9431 0 RCIM_Cubic>]}
       BlueCurve {keys [<at 0.6799 0.999 RCIM_Cubic>,<at 0.8183 0>]}
       GreenCurve {keys [<at 0.6799 0.1882 RCIM_Cubic>,<at 0.8183 0>]}
       RedCurve {keys [<at 0.6799 0.0089 RCIM_Cubic>,<at 0.8183 0>]}
      }})
    R1 Crimson:LinearColor=&ColorFromCurve(ColorCurve=DataInterface() {propertyValues {
       AlphaCurve {keys [<at 0.8979 1 RCIM_Cubic>,<at 0.9431 0 RCIM_Cubic>]}
       BlueCurve {keys [<at 0.6799 0.1933 RCIM_Cubic>,<at 0.8183 0>]}
       GreenCurve {keys [<at 0.6799 0.1944 RCIM_Cubic>,<at 0.8183 0>]}
       RedCurve {keys [<at 0.6799 0.6698 RCIM_Cubic>,<at 0.8183 0>]}
      }})
    R1 DefaultColor=&ColorFromCurve(ColorCurve=DataInterface() {propertyValues {
       AlphaCurve {keys [<at 0.8979 1 RCIM_Cubic>,<at 0.9431 0 RCIM_Cubic>]}
       BlueCurve {keys [<at 0.6799 0.9 RCIM_Cubic>,<at 0.8183 0>]}
       GreenCurve {keys [<at 0.6799 0.8678 RCIM_Cubic>,<at 0.8183 0>]}
       RedCurve {keys [<at 0.6799 0.7904 RCIM_Cubic>,<at 0.8183 0>]}
      }})
    R1 ForestGreen:LinearColor=&ColorFromCurve(ColorCurve=DataInterface() {propertyValues {
       AlphaCurve {keys [<at 0.8979 1 RCIM_Cubic>,<at 0.9431 0 RCIM_Cubic>]}
       BlueCurve {keys [<at 0.6799 0.1528 RCIM_Cubic>,<at 0.8183 0>]}
       GreenCurve {keys [<at 0.6799 0.6406 RCIM_Cubic>,<at 0.8183 0>]}
       RedCurve {keys [<at 0.6799 0.1696 RCIM_Cubic>,<at 0.8183 0>]}
      }})
    R1 Grey:LinearColor=&ColorFromCurve(ColorCurve=DataInterface() {propertyValues {
       AlphaCurve {keys [<at 0.8979 1 RCIM_Cubic>,<at 0.9431 0 RCIM_Cubic>]}
       BlueCurve {keys [<at 0.6799 0.4031 RCIM_Cubic>,<at 0.8183 0>]}
       GreenCurve {keys [<at 0.6799 0.4031 RCIM_Cubic>,<at 0.8183 0>]}
       RedCurve {keys [<at 0.6799 0.4031 RCIM_Cubic>,<at 0.8183 0>]}
      }})
    R1 Lime:LinearColor=&ColorFromCurve(ColorCurve=DataInterface() {propertyValues {
       AlphaCurve {keys [<at 0.8979 1 RCIM_Cubic>,<at 0.9431 0 RCIM_Cubic>]}
       BlueCurve {keys [<at 0.6799 0.1534 RCIM_Cubic>,<at 0.8183 0>]}
       GreenCurve {keys [<at 0.6799 0.7823 RCIM_Cubic>,<at 0.8183 0>]}
       RedCurve {keys [<at 0.6799 0.4327 RCIM_Cubic>,<at 0.8183 0>]}
      }})
    R1 Orange:LinearColor=&ColorFromCurve(ColorCurve=DataInterface() {propertyValues {
       AlphaCurve {keys [<at 0.8979 1 RCIM_Cubic>,<at 0.9431 0 RCIM_Cubic>]}
       BlueCurve {keys [<at 0.6799 0.0549 RCIM_Cubic>,<at 0.8183 0>]}
       GreenCurve {keys [<at 0.6799 0.502 RCIM_Cubic>,<at 0.8183 0>]}
       RedCurve {keys [<at 0.6799 0.7198 RCIM_Cubic>,<at 0.8183 0>]}
      }})
    R1 Pink:LinearColor=&ColorFromCurve(ColorCurve=DataInterface() {propertyValues {
       AlphaCurve {keys [<at 0.8979 1 RCIM_Cubic>,<at 0.9431 0 RCIM_Cubic>]}
       BlueCurve {keys [<at 0.6799 0.8115 RCIM_Cubic>,<at 0.8183 0>]}
       GreenCurve {keys [<at 0.6799 0.4336 RCIM_Cubic>,<at 0.8183 0>]}
       RedCurve {keys [<at 0.6799 0.7635 RCIM_Cubic>,<at 0.8183 0>]}
      }})
    R1 Purple:LinearColor=&ColorFromCurve(ColorCurve=DataInterface() {propertyValues {
       AlphaCurve {keys [<at 0.8979 1 RCIM_Cubic>,<at 0.9431 0 RCIM_Cubic>]}
       BlueCurve {keys [<at 0.6799 0.799 RCIM_Cubic>,<at 0.8183 0>]}
       GreenCurve {keys [<at 0.6799 0.1006 RCIM_Cubic>,<at 0.8183 0>]}
       RedCurve {keys [<at 0.6799 0.334 RCIM_Cubic>,<at 0.8183 0>]}
      }})
    R1 Saffron:LinearColor=&ColorFromCurve(ColorCurve=DataInterface() {propertyValues {
       AlphaCurve {keys [<at 0.8979 1 RCIM_Cubic>,<at 0.9431 0 RCIM_Cubic>]}
       BlueCurve {keys [<at 0.6799 0.0934 RCIM_Cubic>,<at 0.8183 0>]}
       GreenCurve {keys [<at 0.6799 0.562 RCIM_Cubic>,<at 0.8183 0>]}
       RedCurve {keys [<at 0.6799 0.5948 RCIM_Cubic>,<at 0.8183 0>]}
      }})
    R1 SkyBlue:LinearColor=&ColorFromCurve(ColorCurve=DataInterface() {propertyValues {
       AlphaCurve {keys [<at 0.8979 1 RCIM_Cubic>,<at 0.9431 0 RCIM_Cubic>]}
       BlueCurve {keys [<at 0.6799 0.9823 RCIM_Cubic>,<at 0.8183 0>]}
       GreenCurve {keys [<at 0.6799 0.5001 RCIM_Cubic>,<at 0.8183 0>]}
       RedCurve {keys [<at 0.6799 0.0993 RCIM_Cubic>,<at 0.8183 0>]}
      }})
    R1 TitaniumWhite:LinearColor=&ColorFromCurve(ColorCurve=DataInterface() {propertyValues {
       AlphaCurve {keys [<at 0.8979 1 RCIM_Cubic>,<at 0.9431 0 RCIM_Cubic>]}
       BlueCurve {keys [<at 0.6799 0.6105 RCIM_Cubic>,<at 0.8183 0>]}
       GreenCurve {keys [<at 0.6799 0.6105 RCIM_Cubic>,<at 0.8183 0>]}
       RedCurve {keys [<at 0.6799 0.6105 RCIM_Cubic>,<at 0.8183 0>]}
      }})
    R2 Black:LinearColor=&ColorFromCurve(ColorCurve=DataInterface() {propertyValues {
       AlphaCurve {keys [<at 0.0527 0 RCIM_Cubic>,<at 0.144 1 RCIM_Cubic>,<at 0.8465 1>]}
       BlueCurve {keys [<at 0.6966 0>,<at 0.7247 0>]}
       GreenCurve {keys [<at 0.6966 0>,<at 0.7247 0>]}
       RedCurve {keys [<at 0.6966 0>,<at 0.7247 0>]}
      }})
    R2 BurntSienna:LinearColor=&ColorFromCurve(ColorCurve=DataInterface() {propertyValues {
       AlphaCurve {keys [<at 0.0527 0 RCIM_Cubic>,<at 0.144 1 RCIM_Cubic>,<at 0.8465 1>]}
       BlueCurve {keys [<at 0.6966 0>,<at 0.7247 0>]}
       GreenCurve {keys [<at 0.6966 0>,<at 0.7247 0>]}
       RedCurve {keys [<at 0.6966 0>,<at 0.7247 0>]}
      }})
    R2 Cobalt:LinearColor=&ColorFromCurve(ColorCurve=DataInterface() {propertyValues {
       AlphaCurve {keys [<at 0.0527 0 RCIM_Cubic>,<at 0.144 1 RCIM_Cubic>,<at 0.8465 1>]}
       BlueCurve {keys [<at 0.6966 0>,<at 0.7247 0>]}
       GreenCurve {keys [<at 0.6966 0>,<at 0.7247 0>]}
       RedCurve {keys [<at 0.6966 0>,<at 0.7247 0>]}
      }})
    R2 Crimson:LinearColor=&ColorFromCurve(ColorCurve=DataInterface() {propertyValues {
       AlphaCurve {keys [<at 0.0527 0 RCIM_Cubic>,<at 0.144 1 RCIM_Cubic>,<at 0.8465 1>]}
       BlueCurve {keys [<at 0.6966 0>,<at 0.7247 0>]}
       GreenCurve {keys [<at 0.6966 0>,<at 0.7247 0>]}
       RedCurve {keys [<at 0.6966 0>,<at 0.7247 0>]}
      }})
    R2 DefaultColor:LinearColor=&ColorFromCurve(ColorCurve=DataInterface() {propertyValues {
       AlphaCurve {keys [<at 0.0527 0 RCIM_Cubic>,<at 0.144 1 RCIM_Cubic>,<at 0.8465 1>]}
       BlueCurve {keys [<at 0.6966 0>,<at 0.7247 0>]}
       GreenCurve {keys [<at 0.6966 0>,<at 0.7247 0>]}
       RedCurve {keys [<at 0.6966 0>,<at 0.7247 0>]}
      }})
    R2 ForestGreen:LinearColor=&ColorFromCurve(ColorCurve=DataInterface() {propertyValues {
       AlphaCurve {keys [<at 0.0527 0 RCIM_Cubic>,<at 0.144 1 RCIM_Cubic>,<at 0.8465 1>]}
       BlueCurve {keys [<at 0.6966 0>,<at 0.7247 0>]}
       GreenCurve {keys [<at 0.6966 0>,<at 0.7247 0>]}
       RedCurve {keys [<at 0.6966 0>,<at 0.7247 0>]}
      }})
    R2 Gold:LinearColor=&ColorFromCurve(ColorCurve=DataInterface() {propertyValues {
       AlphaCurve {keys [<at 0.0527 0 RCIM_Cubic>,<at 0.144 1 RCIM_Cubic>,<at 0.8465 1>]}
       BlueCurve {keys [<at 0.6966 0>,<at 0.7247 0>]}
       GreenCurve {keys [<at 0.6966 0>,<at 0.7247 0>]}
       RedCurve {keys [<at 0.6966 0>,<at 0.7247 0>]}
      }})
    R2 Grey:LinearColor=&ColorFromCurve(ColorCurve=DataInterface() {propertyValues {
       AlphaCurve {keys [<at 0.0527 0 RCIM_Cubic>,<at 0.144 1 RCIM_Cubic>,<at 0.8465 1>]}
       BlueCurve {keys [<at 0.6966 0>,<at 0.7247 0>]}
       GreenCurve {keys [<at 0.6966 0>,<at 0.7247 0>]}
       RedCurve {keys [<at 0.6966 0>,<at 0.7247 0>]}
      }})
    R2 Lime:LinearColor=&ColorFromCurve(ColorCurve=DataInterface() {propertyValues {
       AlphaCurve {keys [<at 0.0527 0 RCIM_Cubic>,<at 0.144 1 RCIM_Cubic>,<at 0.8465 1>]}
       BlueCurve {keys [<at 0.6966 0>,<at 0.7247 0>]}
       GreenCurve {keys [<at 0.6966 0>,<at 0.7247 0>]}
       RedCurve {keys [<at 0.6966 0>,<at 0.7247 0>]}
      }})
    R2 Orange:LinearColor=&ColorFromCurve(ColorCurve=DataInterface() {propertyValues {
       AlphaCurve {keys [<at 0.0527 0 RCIM_Cubic>,<at 0.144 1 RCIM_Cubic>,<at 0.8465 1>]}
       BlueCurve {keys [<at 0.6966 0>,<at 0.7247 0>]}
       GreenCurve {keys [<at 0.6966 0>,<at 0.7247 0>]}
       RedCurve {keys [<at 0.6966 0>,<at 0.7247 0>]}
      }})
    R2 Pink:LinearColor=&ColorFromCurve(ColorCurve=DataInterface() {propertyValues {
       AlphaCurve {keys [<at 0.0527 0 RCIM_Cubic>,<at 0.144 1 RCIM_Cubic>,<at 0.8465 1>]}
       BlueCurve {keys [<at 0.6966 0>,<at 0.7247 0>]}
       GreenCurve {keys [<at 0.6966 0>,<at 0.7247 0>]}
       RedCurve {keys [<at 0.6966 0>,<at 0.7247 0>]}
      }})
    R2 Purple:LinearColor=&ColorFromCurve(ColorCurve=DataInterface() {propertyValues {
       AlphaCurve {keys [<at 0.0527 0 RCIM_Cubic>,<at 0.144 1 RCIM_Cubic>,<at 0.8465 1>]}
       BlueCurve {keys [<at 0.6966 0>,<at 0.7247 0>]}
       GreenCurve {keys [<at 0.6966 0>,<at 0.7247 0>]}
       RedCurve {keys [<at 0.6966 0>,<at 0.7247 0>]}
      }})
    R2 Saffron:LinearColor=&ColorFromCurve(ColorCurve=DataInterface() {propertyValues {
       AlphaCurve {keys [<at 0.0527 0 RCIM_Cubic>,<at 0.144 1 RCIM_Cubic>,<at 0.8465 1>]}
       BlueCurve {keys [<at 0.6966 0>,<at 0.7247 0>]}
       GreenCurve {keys [<at 0.6966 0>,<at 0.7247 0>]}
       RedCurve {keys [<at 0.6966 0>,<at 0.7247 0>]}
      }})
    R2 SkyBlue:LinearColor=&ColorFromCurve(ColorCurve=DataInterface() {propertyValues {
       AlphaCurve {keys [<at 0.0527 0 RCIM_Cubic>,<at 0.144 1 RCIM_Cubic>,<at 0.8465 1>]}
       BlueCurve {keys [<at 0.6966 0>,<at 0.7247 0>]}
       GreenCurve {keys [<at 0.6966 0>,<at 0.7247 0>]}
       RedCurve {keys [<at 0.6966 0>,<at 0.7247 0>]}
      }})
    R2 TitaniumWhite:LinearColor=&ColorFromCurve(ColorCurve=DataInterface() {propertyValues {
       AlphaCurve {keys [<at 0.0527 0 RCIM_Cubic>,<at 0.144 1 RCIM_Cubic>,<at 0.8465 1>]}
       BlueCurve {keys [<at 0.6966 0>,<at 0.7247 0>]}
       GreenCurve {keys [<at 0.6966 0>,<at 0.7247 0>]}
       RedCurve {keys [<at 0.6966 0>,<at 0.7247 0>]}
      }})
    Saffron=&ColorFromCurve(ColorCurve=DataInterface() {propertyValues {
       AlphaCurve {keys [<at 0.0038 0 RCIM_Cubic>,<at 0.0364 1 RCIM_Cubic>,<at 0.8276 1>,<at 0.8502 0>]}
       BlueCurve {keys [<at 0.6367 0.2492>,<at 0.6543 0>]}
       GreenCurve {keys [<at 0.6367 0.8816>,<at 0.6543 0>]}
       RedCurve {keys [<at 0.6367 0.931>,<at 0.6543 0>]}
      }})
    SkyBlue=&ColorFromCurve(ColorCurve=DataInterface() {propertyValues {
       AlphaCurve {keys [<at 0.0038 0 RCIM_Cubic>,<at 0.0364 1 RCIM_Cubic>,<at 0.8276 1>,<at 0.8502 0>]}
       BlueCurve {keys [<at 0.6367 0.8406>,<at 0.6543 0>]}
       GreenCurve {keys [<at 0.6367 0.3926>,<at 0.6543 0>]}
       RedCurve {keys [<at 0.6367 0.1001>,<at 0.6543 0>]}
      }})
    TitaniumWhite=&ColorFromCurve(ColorCurve=DataInterface() {propertyValues {
       AlphaCurve {keys [<at 0.0038 0 RCIM_Cubic>,<at 0.0364 1 RCIM_Cubic>,<at 0.8276 1>,<at 0.8502 0>]}
       BlueCurve {keys [<at 0.6367 1>,<at 0.6543 0>]}
       GreenCurve {keys [<at 0.6367 1>,<at 0.6543 0>]}
       RedCurve {keys [<at 0.6367 1>,<at 0.6543 0>]}
      }})
    Use R1 Painted Overrides=true
    Use R2 Painted Overrides=true
   ) {module=VC_PaintedVariantsOverride001}
  ]
  rendererClasses [&NiagaraRibbonRendererProperties]
  renderers [
   {rendererClass=&NiagaraRibbonRendererProperties}
   {rendererClass=&NiagaraRibbonRendererProperties, rendererIndex=1}
   {rendererClass=&NiagaraRibbonRendererProperties, rendererIndex=2}
   {rendererClass=&NiagaraRibbonRendererProperties, rendererIndex=3}
  ]
 }
 Ribbon_Trail001 {
  bEnabled=false
  emitterUpdateScript [
   &EmitterState(Inactive Response=::NewEnumerator1, Life Cycle Mode=::NewEnumerator1, Loop Duration=1, UseLoopDelay=false)
   &VC_FilteredSocketCount()
   &VC_RateQualityHandler(Epic Spawn Rate=30, High Spawn Rate=30, Low Spawn Rate=30, Use Custom Rate?=true)
   &VC_SpawnRibbon_v1(Number of Ribbons=$Emitter.PortCount, Spawn Time=-1, SpawnRate=$Emitter.Spawn Rate)
  ]
  particleSpawnScript [
   &VC_PaintedVariantsOverride(
    Black=<rgba 0.01 0.01 0.01 1.25>
    BurntSienna=<rgba 0.0955 0.0295 0.0118 1>
    Cobalt=<rgba 0 0.0066 0.3559 1>
    Crimson=<rgba 0.0816 0 0.0022 1.5>
    ForestGreen=<rgba 0 0.0608 0.0031 1.5>
    Grey=<rgba 0.1649 0.1649 0.1649 0.4435>
    Lime=<rgba 0.0334 0.1565 0.025 1.5>
    Orange=<rgba 0.2205 0.0372 0 1.5>
    Pink=<rgba 0.3003 0.0771 0.2854 1>
    Purple=<rgba 0.0223 0 0.1101 2>
    R1 Black:LinearColor=<rgba 0.0469 0.0469 0.0469 1>
    R1 Cobalt:LinearColor=<rgba 0 0.0185 1 1>
    R1 ForestGreen:LinearColor=<rgba 0 0.2319 0.004 1>
    R1 Grey:LinearColor=<rgba 0.1267 0.1267 0.1267 1>
    R1 Orange:LinearColor=<rgba 0.6724 0.2346 0 1>
    R1 Pink:LinearColor=<rgba 0.5052 0.1552 0.3779 1>
    R1 Purple:LinearColor=<rgba 0.0699 0 0.2552 1>
    R1 SkyBlue:LinearColor=<rgba 0.2058 0.4258 0.5365 1>
    Saffron=<rgba 0.1372 0.1052 0 1>
    SkyBlue=<rgba 0 0.0765 0.1719 1.5>
    TitaniumWhite=<rgba 0.1406 0.1406 0.1406 1>
    Use R1 Painted Overrides=true
   ) {enabled=false}
   &VC_InitializeRibbon_v1(
    Color=$Particles.Color
    Custom Offset=true
    Enable Skeletal Position=true
    Lifetime=&MergeHelper_FloatSwitch(0=1, 1=0.8, 2=0.5, 3=?, Index=$User.GameModeID, Num=::NewEnumerator1)
    Local Vector Offset=&AddVectorToPosition(
     Position=&MergeHelper_VectorSwitch(0=<-3.82 0 0>, 1=<-3.821 0 0>, 2=<-55 0 0>, 3=?, Index=$System.GameModeID, Num=::NewEnumerator1)
     Vector=Linked() {linkedVariable {name=User.DTCarOffset003, type=NiagaraPosition}}
    )
    Offset Amount=1
    Ribbon Width=42
   )
   &AddVelocity(Velocity=<-1500 0 0>) {module=AddVelocity001}
  ]
  particleUpdateScript [
   &ParticleState()
   &InheritSourceMovement(
    Applied Position Scale=&Multiply_VectorByFloat(
     Vector=<1 1 1>
     Float=&FloatFromCurve(FloatCurve=DataInterface() {propertyValues {Curve {keys [{interpMode=RCIM_Cubic, tangentMode=RCTM_Break, time=0, value=0.85}, {interpMode=RCIM_Cubic, tangentMode=RCTM_Break, time=1, value=0}]}}})
    )
    ApplyForce=false
    ApplyPosition=true
    ApplyPositionFalloff=false
   )
   &SolveForcesAndVelocity(Clamp Velocity=false, Limit Acceleration=false)
   /SetVariables(
    Particles.Ribbon 2 Base:LinearColor=&ColorFromCurve(ColorCurve=DataInterface() {propertyValues {
       AlphaCurve {keys [<at 0.8979 1 RCIM_Cubic>,<at 0.9431 0 RCIM_Cubic>]}
       BlueCurve {keys [<at 0.6799 0.9 RCIM_Cubic>,<at 0.8183 0>]}
       GreenCurve {keys [<at 0.6799 0.8678 RCIM_Cubic>,<at 0.8183 0>]}
       RedCurve {keys [<at 0.6799 0.7904 RCIM_Cubic>,<at 0.8183 0>]}
      }})
    Particles.RibbonBase Color:LinearColor=&ColorFromCurve(ColorCurve=DataInterface() {propertyValues {
       AlphaCurve {keys [<at 0.0527 0 RCIM_Cubic>,<at 0.144 1 RCIM_Cubic>,<at 0.8465 1>]}
       BlueCurve {keys [<at 0.6966 0>,<at 0.7247 0>]}
       GreenCurve {keys [<at 0.6966 0>,<at 0.7247 0>]}
       RedCurve {keys [<at 0.6966 0>,<at 0.7247 0>]}
      }})
    Particles.RibbonBase2 Color:LinearColor=&ColorFromCurve(ColorCurve=DataInterface() {propertyValues {
       AlphaCurve {keys [<at 0.0038 0 RCIM_Cubic>,<at 0.0364 1 RCIM_Cubic>,<at 0.8276 1>,<at 0.8502 0>]}
       BlueCurve {keys [<at 0.6367 0>,<at 0.6543 0>]}
       GreenCurve {keys [<at 0.6367 0>,<at 0.6543 0>]}
       RedCurve {keys [<at 0.6367 1.25>,<at 0.6543 0>]}
      }})
    Particles.curve_color_lines:LinearColor=&MultiplyLinearColor(
     Linear Color=$Particles.color_lines
     Scale Factor=&MakeVector4_FromLinearColor(LinearColor=&MakeLinearColorFromVectorAndFloat(
       "Vector (RGB)"=<1 1 1>
       "Float (Alpha)"=&FloatFromCurve(FloatCurve=DataInterface() {propertyValues {Curve {keys [{interpMode=RCIM_Cubic, tangentMode=RCTM_Break, time=0.5, value=1}, {interpMode=RCIM_Cubic, tangentMode=RCTM_Break, time=1, value=1}]}}})
      ))
    )
   ) {isSetParameters=true, module=SetVariables_#5}
   &VC_UpdateRibbon_v1(Enable Skeletal Position=true)
   &Color(Color=&MakeCustomLinearColorFromBool(
     Bool=$User.bIsPainted
     True Color=$Particles.Color
     False Color=&ColorFromCurve(ColorCurve=DataInterface() {propertyValues {
        bExposeCurve=true
        AlphaCurve {keys [<at 0.0095 0>,<at 0.0365 1>,<at 1 1 RCIM_Cubic>]}
        BlueCurve {defaultValue=0, keys [<at 0.8537 1>]}
        GreenCurve {defaultValue=0, keys [<at 0.8537 0.7702>]}
        RedCurve {defaultValue=1, keys [<at 0.8537 0.5989>]}
       }})
    )) {enabled=false}
   &ScaleRibbonWidth(Ribbon Width Scale=&FloatFromCurve(
     FloatCurve=DataInterface() {propertyValues {Curve {keys [
         {arriveTangent=-0.5896, interpMode=RCIM_Cubic, leaveTangent=-0.5896, tangentMode=RCTM_User, time=0.0029, value=0.5874}
         {arriveTangent=1.9541, interpMode=RCIM_Cubic, leaveTangent=1.9541, tangentMode=RCTM_User, time=0.0963, value=0.8133}
         {arriveTangent=2.1445, interpMode=RCIM_Cubic, leaveTangent=2.1445, tangentMode=RCTM_User, time=0.209, value=0.8907}
         {arriveTangent=-0.0014, interpMode=RCIM_Cubic, leaveTangent=-0.0014, tangentMode=RCTM_User, time=0.4609, value=1.0063}
         {
          arriveTangent=-0.5896
          arriveTangentWeight=1.0346
          interpMode=RCIM_Cubic
          leaveTangent=-0.5896
          tangentMode=RCTM_User
          time=0.8245
          value=0.4328
         }
         {arriveTangent=-0.5896, interpMode=RCIM_Cubic, leaveTangent=-0.5896, tangentMode=RCTM_User, time=1.002, value=0.3176}
        ]}}}
     Scale Curve=1.2
    ))
   /SetVariables(Particles.BlueRibbon_Width:NiagaraFloat=&FloatFromCurve(FloatCurve=DataInterface() {propertyValues {Curve {keys [
         {arriveTangent=0.0034, interpMode=RCIM_Cubic, leaveTangent=0.0034, tangentMode=RCTM_User, time=-0.0022, value=42}
         {arriveTangent=-20.8009, interpMode=RCIM_Cubic, leaveTangent=-20.8009, tangentMode=RCTM_User, time=0.5469, value=36.2918}
         {arriveTangent=-0.0044, interpMode=RCIM_Cubic, leaveTangent=-0.0044, tangentMode=RCTM_User, time=1.007, value=15.4938}
        ]}}})) {isSetParameters=true, module=SetVariables_#8}
   /SetVariables(Particles.WPO:NiagaraPosition=&AddVectorToPosition(
     Position=$Particles.Position
     Vector=&TransformVector(
      Vector=&MakeVector(
       X=0
       Y=0
       Z=&Sine(
        Normalized Angle=&Add_Float(
         A=&Multiply_Float(A=$Particles.NormalizedAge, B=-1)
         B=&Multiply_Float(A=$Emitter.LoopedAge, B=4)
        )
        Period=&MaskFloatBySpawnGroup(Masked Float=0.5, Passthrough Float=-0.5)
        Scale=&FloatFromCurve(
         FloatCurve=DataInterface() {propertyValues {Curve {
            defaultValue=1
            keys [
             {interpMode=RCIM_Cubic, tangentMode=RCTM_User, time=0, value=0}
             {arriveTangent=0.4319, interpMode=RCIM_Cubic, leaveTangent=0.4319, time=0.489, value=0.2385}
             {interpMode=RCIM_Cubic, tangentMode=RCTM_User, time=1.0038, value=0.4336}
            ]
           }}}
         Scale Curve=10
        )
        Bias=0
       )
      )
      Destination Space=::Simulation
      Apply Scale=true
     )
    )) {isSetParameters=true, module=SetVariables_#9}
   &DynamicMaterialParameters(
    Index 0 Param 4=10
    Module.Index0Float0Write=true
    Module.Index0Float1Write=true
    Module.Index0Float2Write=true
    Module.Index0Float3Write=true
   )
   /SetVariables(
    Particles.Ribbon 2 Base:LinearColor=&ColorFromCurve(ColorCurve=DataInterface() {propertyValues {
       AlphaCurve {keys [<at 0.8979 1 RCIM_Cubic>,<at 0.9431 0 RCIM_Cubic>]}
       BlueCurve {keys [<at 0.6799 0.9 RCIM_Cubic>,<at 0.8183 0>]}
       GreenCurve {keys [<at 0.6799 0.8678 RCIM_Cubic>,<at 0.8183 0>]}
       RedCurve {keys [<at 0.6799 0.7904 RCIM_Cubic>,<at 0.8183 0>]}
      }})
    Particles.RibbonBase Color:LinearColor=&ColorFromCurve(ColorCurve=DataInterface() {propertyValues {
       AlphaCurve {keys [<at 0.0527 0 RCIM_Cubic>,<at 0.144 1 RCIM_Cubic>,<at 0.8465 1>]}
       BlueCurve {keys [<at 0.6966 0>,<at 0.7247 0>]}
       GreenCurve {keys [<at 0.6966 0>,<at 0.7247 0>]}
       RedCurve {keys [<at 0.6966 0>,<at 0.7247 0>]}
      }})
    Particles.RibbonBase2 Color:LinearColor=&ColorFromCurve(ColorCurve=DataInterface() {propertyValues {
       AlphaCurve {keys [<at 0.0038 0 RCIM_Cubic>,<at 0.0364 1 RCIM_Cubic>,<at 0.8276 1>,<at 0.8502 0>]}
       BlueCurve {keys [<at 0.6367 0>,<at 0.6543 0>]}
       GreenCurve {keys [<at 0.6367 0>,<at 0.6543 0>]}
       RedCurve {keys [<at 0.6367 1.25>,<at 0.6543 0>]}
      }})
    Particles.curve_color_lines:LinearColor=&MultiplyLinearColor(
     Linear Color=$Particles.color_lines
     Scale Factor=&MakeVector4_FromLinearColor(LinearColor=&MakeLinearColorFromVectorAndFloat(
       "Vector (RGB)"=<1 1 1>
       "Float (Alpha)"=&FloatFromCurve(FloatCurve=DataInterface() {propertyValues {Curve {keys [{interpMode=RCIM_Cubic, tangentMode=RCTM_Break, time=0.5, value=1}, {interpMode=RCIM_Cubic, tangentMode=RCTM_Break, time=1, value=1}]}}})
      ))
    )
   ) {isSetParameters=true, module=SetVariables_#7}
  ]
  rendererClasses [&NiagaraRibbonRendererProperties]
  renderers [
   {rendererClass=&NiagaraRibbonRendererProperties}
   {rendererClass=&NiagaraRibbonRendererProperties, rendererIndex=1}
   {rendererClass=&NiagaraRibbonRendererProperties, rendererIndex=2}
   {rendererClass=&NiagaraRibbonRendererProperties, rendererIndex=3}
  ]
 }
 StarSprites {
  emitterUpdateScript [
   &EmitterState()
   &VC_FilteredSocketCount()
   &SpawnRate(SpawnRate=20, Use Spawn Probability=false)
   &SpawnRate(Spawn Probability=10, SpawnGroup=1, SpawnRate=11, Use Spawn Probability=true) {module=SpawnRate003}
  ]
  particleSpawnScript [
   /SetVariables(Particles.RenderTag:NiagaraInt32=&SpawnGroup()) {isSetParameters=true, module=SetVariables_#10}
   &VC_PaintedVariantsOverride(
    Black=<rgba 0.0948 0.0948 0.0948 1>
    BurntSienna=<rgba 0.1656 0.0251 0.014 1>
    Cobalt=<rgba 0 0 0.6562 1>
    Crimson=<rgba 2 0.1104 0.1863 1>
    DefaultColor=<rgba 2 0 0.0615 1>
    ForestGreen=<rgba 0.103 1 0 1>
    Grey=<rgba 0.3115 0.3115 0.3115 1>
    Lime=<rgba 0.3934 1.6979 0 1>
    Orange=<rgba 1.4812 0.1519 0 1>
    Pink=<rgba 2 0.1938 1.5067 1>
    Purple=<rgba 0.1275 0 1.0312 1>
    R1 Black:LinearColor=<rgba 0 0 0 1>
    R1 BurntSienna:LinearColor=<rgba 0 0 0 1>
    R1 Cobalt:LinearColor=<rgba 0 0 0 1>
    R1 Crimson:LinearColor=<rgba 0 0 0 1>
    R1 DefaultColor=<rgba 0 0 0 1>
    R1 ForestGreen:LinearColor=<rgba 0 0 0 1>
    R1 Gold:LinearColor=<rgba 0 0 0 1>
    R1 Grey:LinearColor=<rgba 0 0 0 1>
    R1 Lime:LinearColor=<rgba 0 0 0 1>
    R1 Orange:LinearColor=<rgba 0 0 0 1>
    R1 Pink:LinearColor=<rgba 0 0 0 1>
    R1 Purple:LinearColor=<rgba 0 0 0 1>
    R1 Saffron:LinearColor=<rgba 0 0 0 1>
    R1 SkyBlue:LinearColor=<rgba 0 0 0 1>
    R1 TitaniumWhite:LinearColor=<rgba 0 0 0 1>
    Saffron=<rgba 2 1.8834 0 1>
    SkyBlue=<rgba 0 0.4645 2 1>
    TitaniumWhite=<rgba 0.924 0.924 0.924 1>
    Use R1 Painted Overrides=true
   )
   &InitializeParticle(
    Color=$Particles.Color
    Lifetime Max=&MaskFloatBySpawnGroup(
     Masked Float=&MergeHelper_FloatSwitch(0=0.4, 1=0.4, 2=0.4, 3=?, Index=$User.GameModeID, Num=::NewEnumerator1)
     Passthrough Float=&MergeHelper_FloatSwitch(0=0.4, 1=0.4, 2=0.25, 3=?, Index=$User.GameModeID, Num=::NewEnumerator1)
    )
    Lifetime Min:NiagaraFloat=&MaskFloatBySpawnGroup(
     Masked Float=&MergeHelper_FloatSwitch(0=0.3, 1=0.3, 2=0.3, 3=?, Index=$User.GameModeID, Num=::NewEnumerator1)
     Passthrough Float=&MergeHelper_FloatSwitch(0=0.4, 1=0.4, 2=0.25, 3=?, Index=$User.GameModeID, Num=::NewEnumerator1)
    )
    Lifetime Mode=::NewEnumerator1
    Mesh Scale Mode=::NewEnumerator4
    Sprite Size=<1 0.75>
    Sprite Size Mode=::NewEnumerator4
    UsePositionOffset=false
    Write Material Random=true
   )
   &PartitionParticles(Number Of Partitions=$Emitter.PortCount, Particles Per Partition=2, Partition Mode:ENiagara_PartitionMode=::NewEnumerator1)
   &InitialMeshOrientation(
    Mesh Axis To Orient=<0 0 0>
    Mesh Orientation Mode=::NewEnumerator2
    Rotation=&RandomRangeVector(Minimum=<-1 0 0>, Maximum=<1 0 0>)
    Use Rotation Vector=true
   )
   &SkeletalMeshLocation(
    Bone / Socket Index=$Particles.PartitionParticles.Partition
    Bone / Socket Sampling Mode=::NewEnumerator5
    MaskSpawnGroup=false
    Mesh Sampling Type=::NewEnumerator6
    Sampled Position Offset=&AddVectorToPosition(
     Position=&MaskVectorBySpawnGroup(
      Masked Vector=&MergeHelper_VectorSwitch(0=<-111 0 1>, 1=<-111 0 1>, 2=<5 0 1>, 3=?, Index=$System.GameModeID, Num=::NewEnumerator1)
      Passthrough Vector=&MaskVectorBySpawnGroup(
       Masked Vector=&MergeHelper_VectorSwitch(0=<-90.2 0 1>, 1=<-60 0 1>, 2=<5 0 1>, 3=?, Index=$System.GameModeID, Num=::NewEnumerator1)
       Passthrough Vector=<1 1 1>
       Spawn Group Mask=1
      )
     )
     Vector=Linked() {linkedVariable {name=User.DTCarOffset003, type=NiagaraPosition}}
    )
    Skeletal Mesh=$System.SkeletalMesh
   )
   &ShapeLocation(
    Hemisphere Distribution=<-1 1>
    MaskSpawnGroup=false
    Override Seed=false
    Sphere Radius=&MaskFloatBySpawnGroup(Masked Float=55, Passthrough Float=0)
    Sphere Surface Distribution=1
    UseOwnerScale=true
   )
   &AddVelocity(Velocity=&RandomRangeVector(
     Minimum=&MaskVectorBySpawnGroup(
      Masked Vector=&MergeHelper_VectorSwitch(0=<-750 100 55>, 1=<-750 100 55>, 2=<0 100 55>, 3=?, Index=$System.GameModeID, Num=::NewEnumerator1)
      Passthrough Vector=&MergeHelper_VectorSwitch(0=<-750 0 0>, 1=<-650 0 0>, 2=<0 0 0>, 3=?, Index=$System.GameModeID, Num=::NewEnumerator1)
     )
     Maximum=&MaskVectorBySpawnGroup(
      Masked Vector=&MergeHelper_VectorSwitch(0=<-750 -100 -35>, 1=<-750 -100 -35>, 2=<0 -100 -35>, 3=?, Index=$System.GameModeID, Num=::NewEnumerator1)
      Passthrough Vector=&MergeHelper_VectorSwitch(0=<-750 0 0>, 1=<-650 0 0>, 2=<0 0 0>, 3=?, Index=$System.GameModeID, Num=::NewEnumerator1)
     )
    ))
  ]
  particleUpdateScript [
   &ParticleState()
   &ScaleSpriteSize(
    Non-Uniform Curve Sprite Scale=DataInterface() {propertyValues {
      XCurve {defaultValue=1, keys [{tangentMode=RCTM_User, time=0, value=10}, <at 0.507 10>, {tangentMode=RCTM_User, time=1, value=0}]}
      YCurve {defaultValue=1, keys [{tangentMode=RCTM_User, time=0, value=50}, <at 0.5031 35>, {tangentMode=RCTM_User, time=1, value=0}]}
     }}
    Scale Sprite Size Mode=::NewEnumerator3
   )
   &ScaleMeshSize(Scale Factor=&MaskVectorBySpawnGroup(
     Masked Vector=&VectorFromFloat(Value=&FloatFromCurve(FloatCurve=DataInterface() {propertyValues {Curve {keys [
           <at 0 0.2824>
           {arriveTangent=0.049, interpMode=RCIM_Cubic, leaveTangent=0.049, tangentMode=RCTM_User, time=0.2283, value=0.9882}
           {arriveTangent=-1.3884, interpMode=RCIM_Cubic, leaveTangent=-1.3884, tangentMode=RCTM_User, time=0.6384, value=0.5912}
           {arriveTangent=-1.9181, tangentMode=RCTM_User, time=1, value=0}
          ]}}}))
     Passthrough Vector=<1 1 1>
     Spawn Group Mask=1
    )) {module=ScaleMeshSize001}
   &CurlNoiseForce(
    Bias Noise=false
    Noise Frequency=10
    Noise Strength=&FloatFromCurve(
     FloatCurve=DataInterface() {propertyValues {Curve {keys [<at 0 7500 RCIM_Cubic>, {arriveTangent=-29.3277, interpMode=RCIM_Cubic, leaveTangent=-224.2765, tangentMode=RCTM_Break, time=1, value=0}]}}}
     Scale Curve=&MaskFloatBySpawnGroup(Masked Float=1.1, Passthrough Float=0)
    )
    Pan Noise=false
   )
   &PointAttractionForce(
    Attraction Radius=13.6368
    Attraction Strength=&MaskFloatBySpawnGroup(Masked Float=5000, Passthrough Float=0)
    Attractor Position Offset:Vector3f=<-322.9431 0 0>
    Falloff Exponent=0.3
    Kill Within Radius=false
    Module.Use Falloff=true
    UseAttractorOffset=true
   )
   &Drag(Drag=2, Use Linear Drag=true, Use Rotational Drag=true)
   &SolveForcesAndVelocity(Clamp Velocity=false, Limit Acceleration=false)
   /SetVariables(Particles.Ring Color:LinearColor=<rgba 2 0 0.0615 1>, Particles.SparksColor:LinearColor=<rgba 0 0 0 1>) {enabled=false, isSetParameters=true, module=SetVariables_#11}
   &Color(Color=<rgba 0 0 0 1>) {enabled=false}
   &DynamicMaterialParameters(
    Index 0 Param 1=0.0087
    Index 0 Param 2=-35.9908
    Index 0 Param 3=-0.0477
    Index 1 Param 1=1
    Index 1 Param 2=1
    Module.Index0Float0Write=true
    Module.Index0Float1Write=true
    Module.Index0Float2Write=true
    Module.Index0Float3Write=true
    Module.Index1Float0Write=true
    Module.Index1Float1Write=true
    Module.Index1Float2Write=true
    Module.Index1Float3Write=true
    Write Parameter Index 1=::NewEnumerator1
   )
  ]
  rendererClasses [&NiagaraMeshRendererProperties, &NiagaraSpriteRendererProperties]
  renderers [
   {rendererClass=&NiagaraMeshRendererProperties, rendererIndex=0}
   {rendererClass=&NiagaraSpriteRendererProperties, rendererIndex=1}
  ]
 }
 StarSprites001 {
  bEnabled=false
  emitterUpdateScript [
   &EmitterState()
   &VC_FilteredSocketCount()
   &SpawnRate(SpawnRate=20, Use Spawn Probability=false)
   &SpawnRate(Spawn Probability=10, SpawnGroup=1, SpawnRate=11, Use Spawn Probability=true) {module=SpawnRate003}
  ]
  particleSpawnScript [
   /SetVariables(Particles.RenderTag:NiagaraInt32=&SpawnGroup()) {isSetParameters=true, module=SetVariables_#10}
   &VC_PaintedVariantsOverride(Black=<rgba 0.2309 0.2309 0.2309 0.1652>, BurntSienna=<rgba 0.1823 0.0563 0.0225 0.2696>, Crimson=<rgba 0.2344 0 0.0064 1>) {enabled=false}
   &InitializeParticle(
    Color Mode=::NewEnumerator0
    Lifetime Max=&MaskFloatBySpawnGroup(
     Masked Float=&MergeHelper_FloatSwitch(0=0.4, 1=0.4, 2=0.4, 3=?, Index=$User.GameModeID, Num=::NewEnumerator1)
     Passthrough Float=&MergeHelper_FloatSwitch(0=0.4, 1=0.4, 2=0.25, 3=?, Index=$User.GameModeID, Num=::NewEnumerator1)
    )
    Lifetime Min:NiagaraFloat=&MaskFloatBySpawnGroup(
     Masked Float=&MergeHelper_FloatSwitch(0=0.3, 1=0.3, 2=0.3, 3=?, Index=$User.GameModeID, Num=::NewEnumerator1)
     Passthrough Float=&MergeHelper_FloatSwitch(0=0.4, 1=0.4, 2=0.25, 3=?, Index=$User.GameModeID, Num=::NewEnumerator1)
    )
    Lifetime Mode=::NewEnumerator1
    Mesh Scale Mode=::NewEnumerator4
    Sprite Size=<1 0.75>
    Sprite Size Mode=::NewEnumerator4
    UsePositionOffset=false
    Write Material Random=true
   )
   &PartitionParticles(Number Of Partitions=$Emitter.PortCount, Particles Per Partition=2, Partition Mode:ENiagara_PartitionMode=::NewEnumerator1)
   &InitialMeshOrientation(
    Mesh Axis To Orient=<0 0 0>
    Mesh Orientation Mode=::NewEnumerator2
    Rotation=&RandomRangeVector(Minimum=<-1 0 0>, Maximum=<1 0 0>)
    Use Rotation Vector=true
   )
   &SkeletalMeshLocation(
    Bone / Socket Index=$Particles.PartitionParticles.Partition
    Bone / Socket Sampling Mode=::NewEnumerator5
    MaskSpawnGroup=false
    Mesh Sampling Type=::NewEnumerator6
    Sampled Position Offset=&AddVectorToPosition(
     Position=&MaskVectorBySpawnGroup(
      Masked Vector=&MergeHelper_VectorSwitch(0=<-111 0 1>, 1=<-111 0 1>, 2=<5 0 1>, 3=?, Index=$System.GameModeID, Num=::NewEnumerator1)
      Passthrough Vector=&MaskVectorBySpawnGroup(
       Masked Vector=&MergeHelper_VectorSwitch(0=<-90.2 0 1>, 1=<-60 0 1>, 2=<5 0 1>, 3=?, Index=$System.GameModeID, Num=::NewEnumerator1)
       Passthrough Vector=<1 1 1>
       Spawn Group Mask=1
      )
     )
     Vector=Linked() {linkedVariable {name=User.DTCarOffset003, type=NiagaraPosition}}
    )
    Skeletal Mesh=$System.SkeletalMesh
   )
   &ShapeLocation(
    Hemisphere Distribution=<-1 1>
    MaskSpawnGroup=false
    Override Seed=false
    Sphere Radius=&MaskFloatBySpawnGroup(Masked Float=55, Passthrough Float=0)
    Sphere Surface Distribution=1
    UseOwnerScale=true
   )
   &AddVelocity(Velocity=&RandomRangeVector(
     Minimum=&MaskVectorBySpawnGroup(
      Masked Vector=&MergeHelper_VectorSwitch(0=<-750 100 55>, 1=<-750 100 55>, 2=<0 100 55>, 3=?, Index=$System.GameModeID, Num=::NewEnumerator1)
      Passthrough Vector=&MergeHelper_VectorSwitch(0=<-750 0 0>, 1=<-650 0 0>, 2=<0 0 0>, 3=?, Index=$System.GameModeID, Num=::NewEnumerator1)
     )
     Maximum=&MaskVectorBySpawnGroup(
      Masked Vector=&MergeHelper_VectorSwitch(0=<-750 -100 -35>, 1=<-750 -100 -35>, 2=<0 -100 -35>, 3=?, Index=$System.GameModeID, Num=::NewEnumerator1)
      Passthrough Vector=&MergeHelper_VectorSwitch(0=<-750 0 0>, 1=<-650 0 0>, 2=<0 0 0>, 3=?, Index=$System.GameModeID, Num=::NewEnumerator1)
     )
    ))
  ]
  particleUpdateScript [
   &ParticleState()
   &ScaleSpriteSize(
    Non-Uniform Curve Sprite Scale=DataInterface() {propertyValues {
      XCurve {defaultValue=1, keys [{tangentMode=RCTM_User, time=0, value=10}, <at 0.507 10>, {tangentMode=RCTM_User, time=1, value=0}]}
      YCurve {defaultValue=1, keys [{tangentMode=RCTM_User, time=0, value=50}, <at 0.5031 35>, {tangentMode=RCTM_User, time=1, value=0}]}
     }}
    Scale Sprite Size Mode=::NewEnumerator3
   )
   &ScaleMeshSize(Scale Factor=&MaskVectorBySpawnGroup(
     Masked Vector=&VectorFromFloat(Value=&FloatFromCurve(FloatCurve=DataInterface() {propertyValues {Curve {keys [
           <at 0 0.2824>
           {arriveTangent=0.049, interpMode=RCIM_Cubic, leaveTangent=0.049, tangentMode=RCTM_User, time=0.2283, value=0.9882}
           {arriveTangent=-1.3884, interpMode=RCIM_Cubic, leaveTangent=-1.3884, tangentMode=RCTM_User, time=0.6384, value=0.5912}
           {arriveTangent=-1.9181, tangentMode=RCTM_User, time=1, value=0}
          ]}}}))
     Passthrough Vector=<1 1 1>
     Spawn Group Mask=1
    )) {module=ScaleMeshSize001}
   &CurlNoiseForce(
    Bias Noise=false
    Noise Frequency=10
    Noise Strength=&FloatFromCurve(
     FloatCurve=DataInterface() {propertyValues {Curve {keys [<at 0 7500 RCIM_Cubic>, {arriveTangent=-29.3277, interpMode=RCIM_Cubic, leaveTangent=-224.2765, tangentMode=RCTM_Break, time=1, value=0}]}}}
     Scale Curve=&MaskFloatBySpawnGroup(Masked Float=1.1, Passthrough Float=0)
    )
    Pan Noise=false
   )
   &PointAttractionForce(
    Attraction Radius=13.6368
    Attraction Strength=&MaskFloatBySpawnGroup(Masked Float=5000, Passthrough Float=0)
    Attractor Position Offset:Vector3f=<-322.9431 0 0>
    Falloff Exponent=0.3
    Kill Within Radius=false
    Module.Use Falloff=true
    UseAttractorOffset=true
   )
   &Drag(Drag=2, Use Linear Drag=true, Use Rotational Drag=true)
   &SolveForcesAndVelocity(Clamp Velocity=false, Limit Acceleration=false)
   /SetVariables(Particles.Ring Color:LinearColor=<rgba 2 0 0.0615 1>, Particles.SparksColor:LinearColor=<rgba 0 0 0 1>) {isSetParameters=true, module=SetVariables_#11}
   &Color(Color=<rgba 0 0 0 1>)
   &DynamicMaterialParameters(
    Index 0 Param 1=0.0087
    Index 0 Param 2=-35.9908
    Index 0 Param 3=-0.0477
    Index 1 Param 1=1
    Index 1 Param 2=1
    Module.Index0Float0Write=true
    Module.Index0Float1Write=true
    Module.Index0Float2Write=true
    Module.Index0Float3Write=true
    Module.Index1Float0Write=true
    Module.Index1Float1Write=true
    Module.Index1Float2Write=true
    Module.Index1Float3Write=true
    Write Parameter Index 1=::NewEnumerator1
   )
  ]
  rendererClasses [&NiagaraMeshRendererProperties, &NiagaraSpriteRendererProperties]
  renderers [
   {rendererClass=&NiagaraMeshRendererProperties, rendererIndex=0}
   {rendererClass=&NiagaraSpriteRendererProperties, rendererIndex=1}
  ]
 }
}
summary {
 systemName=NS_Boost_RedPlasma_NEW
 emitters [
  {emitterName=Ribbon_Fire, rendererClasses [&NiagaraRibbonRendererProperties]}
  {emitterName=Ribbon_Trail, rendererClasses [&NiagaraRibbonRendererProperties]}
  {
   bEnabled=false
   emitterName=Meshes001
   rendererClasses [&NiagaraMeshRendererProperties, &NiagaraSpriteRendererProperties]
  }
  {emitterName=StarSprites, rendererClasses [&NiagaraMeshRendererProperties, &NiagaraSpriteRendererProperties]}
  {
   emitterName=NE_VC_Boost_Cone_Spawn001
   rendererClasses [&NiagaraMeshRendererProperties, &NiagaraSpriteRendererProperties]
  }
  {bEnabled=false, emitterName=Ribbon_Fire001, rendererClasses [&NiagaraRibbonRendererProperties]}
  {bEnabled=false, emitterName=Ribbon_Trail001, rendererClasses [&NiagaraRibbonRendererProperties]}
  {bEnabled=false, emitterName=Ribbon_Electricity001, rendererClasses [&NiagaraRibbonRendererProperties]}
  {
   bEnabled=false
   emitterName=StarSprites001
   rendererClasses [&NiagaraMeshRendererProperties, &NiagaraSpriteRendererProperties]
  }
  {
   bEnabled=false
   emitterName=NE_VC_Boost_Cone_Spawn
   rendererClasses [&NiagaraMeshRendererProperties, &NiagaraSpriteRendererProperties]
  }
  {emitterName=Ribbon_Electricity002, rendererClasses [&NiagaraRibbonRendererProperties]}
 ]
 userVariables [
  User.BelowTrailColor {type=LinearColor, description="", defaultValue {struct=LinearColor, value=<rgba 1 1 1 1>}}
  User.bIsFrontendPreview {type=NiagaraBool, defaultValue=true}
  User.bIsPainted {type=NiagaraBool}
  User.ColorVariantID {type=NiagaraInt32}
  User.DTCarOffset {type=NiagaraPosition, description="", defaultValue {}}
  User.DTCarOffset {type=Vector3f, description="", defaultValue {struct=Vector3f, value=<0 0 0>}}
  User.DTCarOffset001 {type=NiagaraPosition, description="", defaultValue {}}
  User.DTCarOffset001 {type=Vector3f, description="", defaultValue {struct=Vector3f, value=<0 0 0>}}
  User.DTCarOffset002 {type=NiagaraPosition, description="", defaultValue {}}
  User.DTCarOffset003 {type=NiagaraPosition, description="", defaultValue {}}
  User.DTCarOffset004 {type=Vector3f, description="", defaultValue {struct=Vector3f, value=<0 0 0>}}
  User.GameModeID {type=NiagaraInt32}
  User.PaintedColor {type=LinearColor, description="", defaultValue {struct=LinearColor, value=<rgba 1 1 1 1>}}
  User.PaintedScalar {type=NiagaraFloat, defaultValue=1, description=""}
  User.PortCount {type=NiagaraInt32, defaultValue=2}
  User.Power_AlphaPaintedR4 {type=NiagaraFloat, defaultValue=1, description=""}
  User.Power_EmissivePaintedR4 {type=NiagaraFloat, defaultValue=1, description=""}
  User.SingleBoostPos {type=Vector3f, description="", defaultValue {struct=Vector3f, value=<0 0 0>}}
  User.WaveMaxScale {type=NiagaraFloat, defaultValue=1, description=""}
  User.WavePeriod {type=NiagaraFloat, defaultValue=1, description=""}
 ]
}
